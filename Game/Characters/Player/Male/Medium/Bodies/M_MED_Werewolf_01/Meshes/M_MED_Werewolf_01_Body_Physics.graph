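# Physics asset for the medium male player Werewolf 01 body mesh, covering two dynamic coat-back bones and the upper spine.
# It has one capsule body on spine_03 and a smaller capsule on each coat bone.
# Each coat bone is constrained directly to spine_03 rather than chained, with identical asymmetric swing limits and a very tight twist limit.
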PhysicsAsset /Game/Characters/Player/Male/Medium/Bodies/M_MED_Werewolf_01/Meshes/M_MED_Werewolf_01_Body_Physics
bodies {
 dyn_coat_back_spine_02 [{shapeType=Capsule, length=2.0624, radius=2.5099, center=<7.4215 0.4279 -1.9805>, rotation=<rot 74.9997 5.0009 0>}]
 dyn_coat_back_spine_new_01 [{shapeType=Capsule, length=2.9547, radius=2.125, center=<7.8236 0.3511 0.3876>, rotation=<rot -90 11.7683 -6.7683>}]
 spine_03 [{shapeType=Capsule, length=14.95, radius=4.4285}]
}
constraints [
 {bone1Name=dyn_coat_back_spine_02, bone2Name=spine_03, swing1LimitDegrees=45, swing2LimitDegrees=20, twistLimitDegrees=1}
 {bone1Name=dyn_coat_back_spine_new_01, bone2Name=spine_03, swing1LimitDegrees=45, swing2LimitDegrees=20, twistLimitDegrees=1}
]
