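# Post-process-style AnimBP for the male werewolf body that adds secondary motion to clothing, copying the attached parent mesh's pose and curves.
# A serial low-alpha AnimDynamics chain jiggles collars, tuft, scraps and shoelaces, with weighted bone constraints and full-alpha dynamics for zippers.
# A spine_03-space RigidBody node using the body physics asset finishes the pose.
# Four coat AnimDynamics nodes are disconnected, right-calf nodes repeat thrice, and back collar alpha is zero.
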AnimBlueprint /Game/Characters/Player/Male/Medium/Bodies/M_MED_Werewolf_01/Meshes/M_MED_Werewolf_01_Body_AnimBP
target_skeleton=&/M_MED_Werewolf_01_Body_Skeleton
graphs [
 AnimGraph {nodes [
   n1 = Anim Dynamics - Bone: dyn_coat_right_spine_03(Alpha=1, ExternalForce="0,0,0") {
    bGravityOverrideInSimSpace=true
    lODThreshold=1
    numSolverIterationsPostUpdate=4
    numSolverIterationsPreUpdate=16
    simulationSpace=Actor
    alphaBoolBlend {blendOption=Linear}
    boundBone {boneName=dyn_coat_right_spine_03}
    componentLinearAccScale=<1 1 1>
    componentLinearVelScale=<1 1 1>
    physicsBodyDefinitions [{
      sphereCollisionRadius=0
      boundBone {boneName=dyn_coat_right_spine_03}
      boxExtents=<5 5 5>
      constraintSetup {angularTargetAxis=AxisZ, twistAxis=AxisZ, angularLimitsMax=<5 30 0>, angularLimitsMin=<-5 -30 0>, angularTarget=<0 0 0>}
      localJointOffset=<0 0 -0.5>
     }]
   }
   n2 = Anim Dynamics - Bone: dyn_coat_left_spine_03(ComponentPose=@n1, Alpha=1, ExternalForce="0,0,0") {
    bGravityOverrideInSimSpace=true
    lODThreshold=1
    numSolverIterationsPostUpdate=4
    numSolverIterationsPreUpdate=16
    simulationSpace=Actor
    alphaBoolBlend {blendOption=Linear}
    boundBone {boneName=dyn_coat_left_spine_03}
    componentLinearAccScale=<1 1 1>
    componentLinearVelScale=<1 1 1>
    physicsBodyDefinitions [{
      sphereCollisionRadius=0
      boundBone {boneName=dyn_coat_left_spine_03}
      boxExtents=<5 5 5>
      constraintSetup {angularTargetAxis=AxisZ, twistAxis=AxisZ, angularLimitsMax=<5 30 0>, angularLimitsMin=<-5 -30 0>, angularTarget=<0 0 0>}
      localJointOffset=<0 0 -0.5>
     }]
   }
   n3 = Anim Dynamics - Bone: dyn_coat_frontRight_spine_03(ComponentPose=@n2, Alpha=1, ExternalForce="0,0,0") {
    bGravityOverrideInSimSpace=true
    lODThreshold=1
    numSolverIterationsPostUpdate=4
    numSolverIterationsPreUpdate=16
    simulationSpace=Actor
    alphaBoolBlend {blendOption=Linear}
    boundBone {boneName=dyn_coat_frontRight_spine_03}
    componentLinearAccScale=<1 1 1>
    componentLinearVelScale=<1 1 1>
    physicsBodyDefinitions [{
      sphereCollisionRadius=0
      boundBone {boneName=dyn_coat_frontRight_spine_03}
      boxExtents=<5 5 5>
      constraintSetup {angularTargetAxis=AxisZ, twistAxis=AxisZ, angularLimitsMax=<5 30 0>, angularLimitsMin=<-5 -30 0>, angularTarget=<0 0 0>}
      localJointOffset=<0 0 -0.5>
     }]
   }
   Anim Dynamics - Bone: dyn_coat_frontLeft_spine_03(ComponentPose=@n3, Alpha=1, ExternalForce="0,0,0") {
    bGravityOverrideInSimSpace=true
    lODThreshold=1
    numSolverIterationsPostUpdate=4
    numSolverIterationsPreUpdate=16
    simulationSpace=Actor
    alphaBoolBlend {blendOption=Linear}
    boundBone {boneName=dyn_coat_frontLeft_spine_03}
    componentLinearAccScale=<1 1 1>
    componentLinearVelScale=<1 1 1>
    physicsBodyDefinitions [{
      sphereCollisionRadius=0
      boundBone {boneName=dyn_coat_frontLeft_spine_03}
      boxExtents=<5 5 5>
      constraintSetup {angularTargetAxis=AxisZ, twistAxis=AxisZ, angularLimitsMax=<5 30 0>, angularLimitsMin=<-5 -30 0>, angularTarget=<0 0 0>}
      localJointOffset=<0 0 -0.5>
     }]
   }
   n4 = Anim Dynamics - Bone: dyn_outer_collar_l(
    ComponentPose=Local To Component(LocalPose=Copy Pose From Mesh() {bCopyCurves=true, bUseAttachedParent=true})
    Alpha=0.125
    ExternalForce="0,0,0"
   ) {
    angularDampingOverride=0.98
    angularSpringConstant=125
    bAngularSpring=true
    bGravityOverrideInSimSpace=true
    bLinearSpring=true
    bOverrideAngularDamping=true
    bOverrideLinearDamping=true
    lODThreshold=1
    linearDampingOverride=0.98
    linearSpringConstant=125
    simulationSpace=Actor
    alphaBoolBlend {blendOption=Linear}
    boundBone {boneName=dyn_outer_collar_l}
    physicsBodyDefinitions [{
      sphereCollisionRadius=0
      boundBone {boneName=dyn_outer_collar_l}
      boxExtents=<30 30 30>
      constraintSetup {
       angularTargetAxis=AxisY
       angularLimitsMax=<0 15 15>
       angularLimitsMin=<0 -15 -15>
       angularTarget=<0 0 0>
       linearAxesMax=<10 10 10>
       linearAxesMin=<0 -10 -10>
      }
      localJointOffset=<0 0 0>
     }]
   }
   n5 = Anim Dynamics - Bone: dyn_outer_collar_r(ComponentPose=@n4, Alpha=0.125, ExternalForce="0,0,0") {
    angularDampingOverride=0.98
    angularSpringConstant=125
    bAngularSpring=true
    bGravityOverrideInSimSpace=true
    bLinearSpring=true
    bOverrideAngularDamping=true
    bOverrideLinearDamping=true
    lODThreshold=1
    linearDampingOverride=0.98
    linearSpringConstant=125
    simulationSpace=Actor
    alphaBoolBlend {blendOption=Linear}
    boundBone {boneName=dyn_outer_collar_r}
    physicsBodyDefinitions [{
      sphereCollisionRadius=0
      boundBone {boneName=dyn_outer_collar_r}
      boxExtents=<30 30 30>
      constraintSetup {
       angularTargetAxis=AxisY
       angularLimitsMax=<0 15 15>
       angularLimitsMin=<0 -15 -15>
       angularTarget=<0 0 0>
       linearAxesMax=<10 10 10>
       linearAxesMin=<0 -10 -10>
      }
      localJointOffset=<0 0 0>
     }]
   }
   n6 = Anim Dynamics - Bone: dyn_outer_collar_back(ComponentPose=@n5, Alpha=0, ExternalForce="0,0,0") {
    angularDampingOverride=0.98
    angularSpringConstant=80
    bAngularSpring=true
    bGravityOverrideInSimSpace=true
    bLinearSpring=true
    bOverrideAngularDamping=true
    bOverrideLinearDamping=true
    lODThreshold=1
    linearDampingOverride=0.98
    linearSpringConstant=80
    simulationSpace=Actor
    alphaBoolBlend {blendOption=Linear}
    boundBone {boneName=dyn_outer_collar_back}
    physicsBodyDefinitions [{
      sphereCollisionRadius=0
      boundBone {boneName=dyn_outer_collar_back}
      boxExtents=<30 30 30>
      constraintSetup {
       angularTargetAxis=AxisY
       angularLimitsMax=<0 15 15>
       angularLimitsMin=<0 -15 -15>
       angularTarget=<0 0 0>
       linearAxesMax=<10 10 10>
       linearAxesMin=<0 -10 -10>
      }
      localJointOffset=<0 0 0>
     }]
   }
   n7 = "Constraint - const_zipper (PPP)"!(ComponentPose=@n6, Alpha=1, ConstraintWeights_0=0.0816, ConstraintWeights_1:float=0, ConstraintWeights_2:float=0.9184) -> (Pose) {
    alphaBoolBlend {blendOption=Linear}
    boneToModify {boneName=const_zipper}
    constraintSetup [{targetBone {boneName=spine_01}}, {targetBone {boneName=thigh_twist_01_l}}, {targetBone {boneName=pelvis}}]
    constraintWeights [1,1,1]
   }
   n8 = "Constraint - dyn_neck_zipper_01 (PPPPPP)"!(
    ComponentPose=@n7
    Alpha=1
    ConstraintWeights_0=0.7142
    ConstraintWeights_1:float=0.003
    ConstraintWeights_2:float=0.0454
    ConstraintWeights_3:float=0.0017
    ConstraintWeights_4:float=0.1493
    ConstraintWeights_5:float=0.0865
   ) -> (Pose) {
    alphaBoolBlend {blendOption=Linear}
    boneToModify {boneName=dyn_neck_zipper_01}
    constraintSetup [
     {targetBone {boneName=head}}
     {targetBone {boneName=clavicle_l}}
     {targetBone {boneName=spine_05}}
     {targetBone {boneName=clavicle_r}}
     {targetBone {boneName=neck_02}}
     {targetBone {boneName=neck_01}}
    ]
    constraintWeights [1,1,1,1,1,1]
   }
   n9 = Anim Dynamics - Bone: dyn_neck_zipper_02(ComponentPose=@n8, Alpha=1, ExternalForce="0,0,0") {
    bGravityOverrideInSimSpace=true
    lODThreshold=1
    numSolverIterationsPostUpdate=4
    numSolverIterationsPreUpdate=16
    simulationSpace=Actor
    alphaBoolBlend {blendOption=Linear}
    boundBone {boneName=dyn_neck_zipper_02}
    componentLinearAccScale=<1 1 1>
    componentLinearVelScale=<1 1 1>
    physicsBodyDefinitions [{
      sphereCollisionRadius=0
      boundBone {boneName=dyn_neck_zipper_02}
      boxExtents=<5 5 5>
      constraintSetup {angularTargetAxis=AxisZ, twistAxis=AxisZ, angularLimitsMax=<5 30 0>, angularLimitsMin=<-5 -30 0>, angularTarget=<0 0 0>}
      localJointOffset=<0 0 -0.5>
     }]
   }
   n10 = Anim Dynamics - Bone: dyn_zipper(ComponentPose=@n9, Alpha=1) {
    bGravityOverrideInSimSpace=true
    lODThreshold=1
    numSolverIterationsPostUpdate=4
    numSolverIterationsPreUpdate=16
    simulationSpace=Actor
    alphaBoolBlend {blendOption=Linear}
    boundBone {boneName=dyn_zipper}
    componentLinearAccScale=<1 1 1>
    componentLinearVelScale=<1 1 1>
    physicsBodyDefinitions [{
      sphereCollisionRadius=0
      boundBone {boneName=dyn_zipper}
      boxExtents=<5 5 5>
      constraintSetup {angularTargetAxis=AxisZ, twistAxis=AxisZ, angularLimitsMax=<5 30 0>, angularLimitsMin=<-5 -30 0>, angularTarget=<0 0 0>}
      localJointOffset=<0 0 -0.5>
     }]
   }
   n11 = Anim Dynamics - Bone: dyn_inner_collar_l(ComponentPose=@n10, Alpha=0.125, ExternalForce="0,0,0") {
    angularDampingOverride=0.98
    angularSpringConstant=125
    bAngularSpring=true
    bGravityOverrideInSimSpace=true
    bLinearSpring=true
    bOverrideAngularDamping=true
    bOverrideLinearDamping=true
    lODThreshold=1
    linearDampingOverride=0.98
    linearSpringConstant=125
    simulationSpace=Actor
    alphaBoolBlend {blendOption=Linear}
    boundBone {boneName=dyn_inner_collar_l}
    physicsBodyDefinitions [{
      sphereCollisionRadius=0
      boundBone {boneName=dyn_inner_collar_l}
      boxExtents=<30 30 30>
      constraintSetup {
       angularTargetAxis=AxisY
       angularLimitsMax=<0 15 15>
       angularLimitsMin=<0 -15 -15>
       angularTarget=<0 0 0>
       linearAxesMax=<10 10 10>
       linearAxesMin=<0 -10 -10>
      }
      localJointOffset=<0 0 0>
     }]
   }
   n12 = Anim Dynamics - Bone: dyn_inner_collar_r(ComponentPose=@n11, Alpha=0.125, ExternalForce="0,0,0") {
    angularDampingOverride=0.98
    angularSpringConstant=125
    bAngularSpring=true
    bGravityOverrideInSimSpace=true
    bLinearSpring=true
    bOverrideAngularDamping=true
    bOverrideLinearDamping=true
    lODThreshold=1
    linearDampingOverride=0.98
    linearSpringConstant=125
    simulationSpace=Actor
    alphaBoolBlend {blendOption=Linear}
    boundBone {boneName=dyn_inner_collar_r}
    physicsBodyDefinitions [{
      sphereCollisionRadius=0
      boundBone {boneName=dyn_inner_collar_r}
      boxExtents=<30 30 30>
      constraintSetup {
       angularTargetAxis=AxisY
       angularLimitsMax=<0 15 15>
       angularLimitsMin=<0 -15 -15>
       angularTarget=<0 0 0>
       linearAxesMax=<10 10 10>
       linearAxesMin=<0 -10 -10>
      }
      localJointOffset=<0 0 0>
     }]
   }
   n13 = Anim Dynamics - Bone: dyn_coat_tuft_l(ComponentPose=@n12, Alpha=0.2, ExternalForce="0,0,0") {
    bGravityOverrideInSimSpace=true
    lODThreshold=1
    numSolverIterationsPostUpdate=4
    numSolverIterationsPreUpdate=16
    alphaBoolBlend {blendOption=Linear}
    boundBone {boneName=dyn_coat_tuft_l}
    componentLinearAccScale=<1 1 1>
    componentLinearVelScale=<1 1 1>
    physicsBodyDefinitions [{
      sphereCollisionRadius=0
      boundBone {boneName=dyn_coat_tuft_l}
      boxExtents=<10 20 10>
      constraintSetup {angularLimitsMax=<15 15 0>, angularLimitsMin=<-15 -15 0>, angularTarget=<0 0 0>, linearAxesMin=<-5 0 0>}
      localJointOffset=<10 0 -5>
     }]
   }
   n14 = Anim Dynamics - Bone: dyn_r_thigh_scrap_outer(ComponentPose=@n13, Alpha=0.125, ExternalForce="0,0,0") {
    angularSpringConstant=80
    bAngularSpring=true
    bGravityOverrideInSimSpace=true
    bLinearSpring=true
    gravityScale=0
    lODThreshold=1
    linearSpringConstant=80
    simulationSpace=Actor
    alphaBoolBlend {blendOption=Linear}
    boundBone {boneName=dyn_r_thigh_scrap_outer}
    physicsBodyDefinitions [{
      sphereCollisionRadius=0
      boundBone {boneName=dyn_r_thigh_scrap_outer}
      boxExtents=<30 30 30>
      constraintSetup {
       angularTargetAxis=AxisY
       angularLimitsMax=<0 30 30>
       angularLimitsMin=<0 -30 -30>
       angularTarget=<0 0 0>
       linearAxesMax=<10 20 20>
       linearAxesMin=<0 -20 -20>
      }
      localJointOffset=<0 0 0>
     }]
   }
   n15 = Anim Dynamics - Bone: dyn_r_calf_scrap_outer(ComponentPose=@n14, Alpha=0.06, ExternalForce="0,0,0") {
    angularDampingOverride=0.98
    angularSpringConstant=30
    bAngularSpring=true
    bGravityOverrideInSimSpace=true
    bLinearSpring=true
    bOverrideAngularDamping=true
    bOverrideLinearDamping=true
    gravityScale=0
    lODThreshold=1
    linearDampingOverride=0.98
    linearSpringConstant=30
    simulationSpace=Actor
    alphaBoolBlend {blendOption=Linear}
    boundBone {boneName=dyn_r_calf_scrap_outer}
    physicsBodyDefinitions [{
      sphereCollisionRadius=0
      boundBone {boneName=dyn_r_calf_scrap_outer}
      boxExtents=<30 30 30>
      constraintSetup {
       angularTargetAxis=AxisY
       angularLimitsMax=<0 30 30>
       angularLimitsMin=<0 -30 -30>
       angularTarget=<0 0 0>
       linearAxesMax=<10 20 20>
       linearAxesMin=<0 -20 -20>
      }
      localJointOffset=<0 0 0>
     }]
   }
   n16 = Anim Dynamics - Bone: dyn_r_calf_scrap_front(ComponentPose=@n15, Alpha=0.06, ExternalForce="0,0,0") {
    angularDampingOverride=0.98
    angularSpringConstant=30
    bAngularSpring=true
    bGravityOverrideInSimSpace=true
    bLinearSpring=true
    bOverrideAngularDamping=true
    bOverrideLinearDamping=true
    gravityScale=0
    lODThreshold=1
    linearDampingOverride=0.98
    linearSpringConstant=30
    simulationSpace=Actor
    alphaBoolBlend {blendOption=Linear}
    boundBone {boneName=dyn_r_calf_scrap_front}
    physicsBodyDefinitions [{
      sphereCollisionRadius=0
      boundBone {boneName=dyn_r_calf_scrap_front}
      boxExtents=<30 30 30>
      constraintSetup {
       angularTargetAxis=AxisY
       angularLimitsMax=<0 30 30>
       angularLimitsMin=<0 -30 -30>
       angularTarget=<0 0 0>
       linearAxesMax=<10 20 20>
       linearAxesMin=<0 -20 -20>
      }
      localJointOffset=<0 0 0>
     }]
   }
   n17 = Anim Dynamics - Bone: dyyn_r_calf_scrap_back(ComponentPose=@n16, Alpha=0.06, ExternalForce="0,0,0") {
    angularDampingOverride=0.98
    angularSpringConstant=30
    bAngularSpring=true
    bGravityOverrideInSimSpace=true
    bLinearSpring=true
    bOverrideAngularDamping=true
    bOverrideLinearDamping=true
    gravityScale=0
    lODThreshold=1
    linearDampingOverride=0.98
    linearSpringConstant=30
    simulationSpace=Actor
    alphaBoolBlend {blendOption=Linear}
    boundBone {boneName=dyyn_r_calf_scrap_back}
    physicsBodyDefinitions [{
      sphereCollisionRadius=0
      boundBone {boneName=dyyn_r_calf_scrap_back}
      boxExtents=<30 30 30>
      constraintSetup {
       angularTargetAxis=AxisY
       angularLimitsMax=<0 30 30>
       angularLimitsMin=<0 -30 -30>
       angularTarget=<0 0 0>
       linearAxesMax=<10 20 20>
       linearAxesMin=<0 -20 -20>
      }
      localJointOffset=<0 0 0>
     }]
   }
   n18 = Anim Dynamics - Bone: dyn_r_calf_scrap_outer(ComponentPose=@n17, Alpha=0.06, ExternalForce="0,0,0") {
    angularDampingOverride=0.98
    angularSpringConstant=30
    bAngularSpring=true
    bGravityOverrideInSimSpace=true
    bLinearSpring=true
    bOverrideAngularDamping=true
    bOverrideLinearDamping=true
    gravityScale=0
    lODThreshold=1
    linearDampingOverride=0.98
    linearSpringConstant=30
    simulationSpace=Actor
    alphaBoolBlend {blendOption=Linear}
    boundBone {boneName=dyn_r_calf_scrap_outer}
    physicsBodyDefinitions [{
      sphereCollisionRadius=0
      boundBone {boneName=dyn_r_calf_scrap_outer}
      boxExtents=<30 30 30>
      constraintSetup {
       angularTargetAxis=AxisY
       angularLimitsMax=<0 30 30>
       angularLimitsMin=<0 -30 -30>
       angularTarget=<0 0 0>
       linearAxesMax=<10 20 20>
       linearAxesMin=<0 -20 -20>
      }
      localJointOffset=<0 0 0>
     }]
   }
   n19 = Anim Dynamics - Bone: dyn_r_calf_scrap_front(ComponentPose=@n18, Alpha=0.06, ExternalForce="0,0,0") {
    angularDampingOverride=0.98
    angularSpringConstant=30
    bAngularSpring=true
    bGravityOverrideInSimSpace=true
    bLinearSpring=true
    bOverrideAngularDamping=true
    bOverrideLinearDamping=true
    gravityScale=0
    lODThreshold=1
    linearDampingOverride=0.98
    linearSpringConstant=30
    simulationSpace=Actor
    alphaBoolBlend {blendOption=Linear}
    boundBone {boneName=dyn_r_calf_scrap_front}
    physicsBodyDefinitions [{
      sphereCollisionRadius=0
      boundBone {boneName=dyn_r_calf_scrap_front}
      boxExtents=<30 30 30>
      constraintSetup {
       angularTargetAxis=AxisY
       angularLimitsMax=<0 30 30>
       angularLimitsMin=<0 -30 -30>
       angularTarget=<0 0 0>
       linearAxesMax=<10 20 20>
       linearAxesMin=<0 -20 -20>
      }
      localJointOffset=<0 0 0>
     }]
   }
   n20 = Anim Dynamics - Bone: dyyn_r_calf_scrap_back(ComponentPose=@n19, Alpha=0.06, ExternalForce="0,0,0") {
    angularDampingOverride=0.98
    angularSpringConstant=30
    bAngularSpring=true
    bGravityOverrideInSimSpace=true
    bLinearSpring=true
    bOverrideAngularDamping=true
    bOverrideLinearDamping=true
    gravityScale=0
    lODThreshold=1
    linearDampingOverride=0.98
    linearSpringConstant=30
    simulationSpace=Actor
    alphaBoolBlend {blendOption=Linear}
    boundBone {boneName=dyyn_r_calf_scrap_back}
    physicsBodyDefinitions [{
      sphereCollisionRadius=0
      boundBone {boneName=dyyn_r_calf_scrap_back}
      boxExtents=<30 30 30>
      constraintSetup {
       angularTargetAxis=AxisY
       angularLimitsMax=<0 30 30>
       angularLimitsMin=<0 -30 -30>
       angularTarget=<0 0 0>
       linearAxesMax=<10 20 20>
       linearAxesMin=<0 -20 -20>
      }
      localJointOffset=<0 0 0>
     }]
   }
   n21 = Anim Dynamics - Bone: dyn_l_leg_scrap_outer(ComponentPose=@n20, Alpha=0.06, ExternalForce="0,0,0") {
    angularDampingOverride=0.98
    angularSpringConstant=30
    bAngularSpring=true
    bGravityOverrideInSimSpace=true
    bLinearSpring=true
    bOverrideAngularDamping=true
    bOverrideLinearDamping=true
    gravityScale=0
    lODThreshold=1
    linearDampingOverride=0.98
    linearSpringConstant=30
    simulationSpace=Actor
    alphaBoolBlend {blendOption=Linear}
    boundBone {boneName=dyn_l_leg_scrap_outer}
    physicsBodyDefinitions [{
      sphereCollisionRadius=0
      boundBone {boneName=dyn_l_leg_scrap_outer}
      boxExtents=<30 30 30>
      constraintSetup {
       angularTargetAxis=AxisY
       angularLimitsMax=<0 30 30>
       angularLimitsMin=<0 -30 -30>
       angularTarget=<0 0 0>
       linearAxesMax=<10 20 20>
       linearAxesMin=<0 -20 -20>
      }
      localJointOffset=<0 0 0>
     }]
   }
   n22 = Anim Dynamics - Bone: dyn_l_leg_scrap_inner(ComponentPose=@n21, Alpha=0.06, ExternalForce="0,0,0") {
    angularDampingOverride=0.98
    angularSpringConstant=30
    bAngularSpring=true
    bGravityOverrideInSimSpace=true
    bLinearSpring=true
    bOverrideAngularDamping=true
    bOverrideLinearDamping=true
    gravityScale=0
    lODThreshold=1
    linearDampingOverride=0.98
    linearSpringConstant=30
    simulationSpace=Actor
    alphaBoolBlend {blendOption=Linear}
    boundBone {boneName=dyn_l_leg_scrap_inner}
    physicsBodyDefinitions [{
      sphereCollisionRadius=0
      boundBone {boneName=dyn_l_leg_scrap_inner}
      boxExtents=<30 30 30>
      constraintSetup {
       angularTargetAxis=AxisY
       angularLimitsMax=<0 30 30>
       angularLimitsMin=<0 -30 -30>
       angularTarget=<0 0 0>
       linearAxesMax=<10 20 20>
       linearAxesMin=<0 -20 -20>
      }
      localJointOffset=<0 0 0>
     }]
   }
   n23 = Anim Dynamics - Bone: dyn_r_calf_scrap_outer(ComponentPose=@n22, Alpha=0.06, ExternalForce="0,0,0") {
    angularDampingOverride=0.98
    angularSpringConstant=30
    bAngularSpring=true
    bGravityOverrideInSimSpace=true
    bLinearSpring=true
    bOverrideAngularDamping=true
    bOverrideLinearDamping=true
    gravityScale=0
    lODThreshold=1
    linearDampingOverride=0.98
    linearSpringConstant=30
    simulationSpace=Actor
    alphaBoolBlend {blendOption=Linear}
    boundBone {boneName=dyn_r_calf_scrap_outer}
    physicsBodyDefinitions [{
      sphereCollisionRadius=0
      boundBone {boneName=dyn_r_calf_scrap_outer}
      boxExtents=<30 30 30>
      constraintSetup {
       angularTargetAxis=AxisY
       angularLimitsMax=<0 30 30>
       angularLimitsMin=<0 -30 -30>
       angularTarget=<0 0 0>
       linearAxesMax=<10 20 20>
       linearAxesMin=<0 -20 -20>
      }
      localJointOffset=<0 0 0>
     }]
   }
   n24 = Anim Dynamics - Bone: dyn_r_calf_scrap_front(ComponentPose=@n23, Alpha=0.06, ExternalForce="0,0,0") {
    angularDampingOverride=0.98
    angularSpringConstant=30
    bAngularSpring=true
    bGravityOverrideInSimSpace=true
    bLinearSpring=true
    bOverrideAngularDamping=true
    bOverrideLinearDamping=true
    gravityScale=0
    lODThreshold=1
    linearDampingOverride=0.98
    linearSpringConstant=30
    simulationSpace=Actor
    alphaBoolBlend {blendOption=Linear}
    boundBone {boneName=dyn_r_calf_scrap_front}
    physicsBodyDefinitions [{
      sphereCollisionRadius=0
      boundBone {boneName=dyn_r_calf_scrap_front}
      boxExtents=<30 30 30>
      constraintSetup {
       angularTargetAxis=AxisY
       angularLimitsMax=<0 30 30>
       angularLimitsMin=<0 -30 -30>
       angularTarget=<0 0 0>
       linearAxesMax=<10 20 20>
       linearAxesMin=<0 -20 -20>
      }
      localJointOffset=<0 0 0>
     }]
   }
   n25 = Anim Dynamics - Bone: dyyn_r_calf_scrap_back(ComponentPose=@n24, Alpha=0.06, ExternalForce="0,0,0") {
    angularDampingOverride=0.98
    angularSpringConstant=30
    bAngularSpring=true
    bGravityOverrideInSimSpace=true
    bLinearSpring=true
    bOverrideAngularDamping=true
    bOverrideLinearDamping=true
    gravityScale=0
    lODThreshold=1
    linearDampingOverride=0.98
    linearSpringConstant=30
    simulationSpace=Actor
    alphaBoolBlend {blendOption=Linear}
    boundBone {boneName=dyyn_r_calf_scrap_back}
    physicsBodyDefinitions [{
      sphereCollisionRadius=0
      boundBone {boneName=dyyn_r_calf_scrap_back}
      boxExtents=<30 30 30>
      constraintSetup {
       angularTargetAxis=AxisY
       angularLimitsMax=<0 30 30>
       angularLimitsMin=<0 -30 -30>
       angularTarget=<0 0 0>
       linearAxesMax=<10 20 20>
       linearAxesMin=<0 -20 -20>
      }
      localJointOffset=<0 0 0>
     }]
   }
   n26 = Anim Dynamics - Bone: dyn_r_shoelace_outer_inner(ComponentPose=@n25, Alpha=0.06, ExternalForce="0,0,0") {
    angularDampingOverride=0.98
    angularSpringConstant=30
    bAngularSpring=true
    bGravityOverrideInSimSpace=true
    bLinearSpring=true
    bOverrideAngularDamping=true
    bOverrideLinearDamping=true
    gravityScale=0
    lODThreshold=1
    linearDampingOverride=0.98
    linearSpringConstant=30
    simulationSpace=Actor
    alphaBoolBlend {blendOption=Linear}
    boundBone {boneName=dyn_r_shoelace_outer_inner}
    physicsBodyDefinitions [{
      sphereCollisionRadius=0
      boundBone {boneName=dyn_r_shoelace_outer_inner}
      boxExtents=<30 30 30>
      constraintSetup {
       angularTargetAxis=AxisY
       angularLimitsMax=<0 30 30>
       angularLimitsMin=<0 -30 -30>
       angularTarget=<0 0 0>
       linearAxesMax=<10 20 20>
       linearAxesMin=<0 -20 -20>
      }
      localJointOffset=<0 0 0>
     }]
   }
   n27 = Anim Dynamics - Bone: dyn_r_shoelace_outer(ComponentPose=@n26, Alpha=0.06, ExternalForce="0,0,0") {
    angularDampingOverride=0.98
    angularSpringConstant=30
    bAngularSpring=true
    bGravityOverrideInSimSpace=true
    bLinearSpring=true
    bOverrideAngularDamping=true
    bOverrideLinearDamping=true
    gravityScale=0
    lODThreshold=1
    linearDampingOverride=0.98
    linearSpringConstant=30
    simulationSpace=Actor
    alphaBoolBlend {blendOption=Linear}
    boundBone {boneName=dyn_r_shoelace_outer}
    physicsBodyDefinitions [{
      sphereCollisionRadius=0
      boundBone {boneName=dyn_r_shoelace_outer}
      boxExtents=<30 30 30>
      constraintSetup {
       angularTargetAxis=AxisY
       angularLimitsMax=<0 30 30>
       angularLimitsMin=<0 -30 -30>
       angularTarget=<0 0 0>
       linearAxesMax=<10 20 20>
       linearAxesMin=<0 -20 -20>
      }
      localJointOffset=<0 0 0>
     }]
   }
   n28 = Anim Dynamics - Bone: dyn_r_shoelace_inner(ComponentPose=@n27, Alpha=0.06, ExternalForce="0,0,0") {
    angularDampingOverride=0.98
    angularSpringConstant=30
    bAngularSpring=true
    bGravityOverrideInSimSpace=true
    bLinearSpring=true
    bOverrideAngularDamping=true
    bOverrideLinearDamping=true
    gravityScale=0
    lODThreshold=1
    linearDampingOverride=0.98
    linearSpringConstant=30
    simulationSpace=Actor
    alphaBoolBlend {blendOption=Linear}
    boundBone {boneName=dyn_r_shoelace_inner}
    physicsBodyDefinitions [{
      sphereCollisionRadius=0
      boundBone {boneName=dyn_r_shoelace_inner}
      boxExtents=<30 30 30>
      constraintSetup {
       angularTargetAxis=AxisY
       angularLimitsMax=<0 30 30>
       angularLimitsMin=<0 -30 -30>
       angularTarget=<0 0 0>
       linearAxesMax=<10 20 20>
       linearAxesMin=<0 -20 -20>
      }
      localJointOffset=<0 0 0>
     }]
   }
   n29 = Anim Dynamics - Bone: dyn_l_shoelace_outer_inner(ComponentPose=@n28, Alpha=0.06, ExternalForce="0,0,0") {
    angularDampingOverride=0.98
    angularSpringConstant=30
    bAngularSpring=true
    bGravityOverrideInSimSpace=true
    bLinearSpring=true
    bOverrideAngularDamping=true
    bOverrideLinearDamping=true
    gravityScale=0
    lODThreshold=1
    linearDampingOverride=0.98
    linearSpringConstant=30
    simulationSpace=Actor
    alphaBoolBlend {blendOption=Linear}
    boundBone {boneName=dyn_l_shoelace_outer_inner}
    physicsBodyDefinitions [{
      sphereCollisionRadius=0
      boundBone {boneName=dyn_l_shoelace_outer_inner}
      boxExtents=<30 30 30>
      constraintSetup {
       angularTargetAxis=AxisY
       angularLimitsMax=<0 30 30>
       angularLimitsMin=<0 -30 -30>
       angularTarget=<0 0 0>
       linearAxesMax=<10 20 20>
       linearAxesMin=<0 -20 -20>
      }
      localJointOffset=<0 0 0>
     }]
   }
   n30 = Anim Dynamics - Bone: dyn_l_shoelace_outer(ComponentPose=@n29, Alpha=0.06, ExternalForce="0,0,0") {
    angularDampingOverride=0.98
    angularSpringConstant=30
    bAngularSpring=true
    bGravityOverrideInSimSpace=true
    bLinearSpring=true
    bOverrideAngularDamping=true
    bOverrideLinearDamping=true
    gravityScale=0
    lODThreshold=1
    linearDampingOverride=0.98
    linearSpringConstant=30
    simulationSpace=Actor
    alphaBoolBlend {blendOption=Linear}
    boundBone {boneName=dyn_l_shoelace_outer}
    physicsBodyDefinitions [{
      sphereCollisionRadius=0
      boundBone {boneName=dyn_l_shoelace_outer}
      boxExtents=<30 30 30>
      constraintSetup {
       angularTargetAxis=AxisY
       angularLimitsMax=<0 30 30>
       angularLimitsMin=<0 -30 -30>
       angularTarget=<0 0 0>
       linearAxesMax=<10 20 20>
       linearAxesMin=<0 -20 -20>
      }
      localJointOffset=<0 0 0>
     }]
   }
   n31 = Anim Dynamics - Bone: dyn_l_shoelace_inner(ComponentPose=@n30, Alpha=0.06, ExternalForce="0,0,0") {
    angularDampingOverride=0.98
    angularSpringConstant=30
    bAngularSpring=true
    bGravityOverrideInSimSpace=true
    bLinearSpring=true
    bOverrideAngularDamping=true
    bOverrideLinearDamping=true
    gravityScale=0
    lODThreshold=1
    linearDampingOverride=0.98
    linearSpringConstant=30
    simulationSpace=Actor
    alphaBoolBlend {blendOption=Linear}
    boundBone {boneName=dyn_l_shoelace_inner}
    physicsBodyDefinitions [{
      sphereCollisionRadius=0
      boundBone {boneName=dyn_l_shoelace_inner}
      boxExtents=<30 30 30>
      constraintSetup {
       angularTargetAxis=AxisY
       angularLimitsMax=<0 30 30>
       angularLimitsMin=<0 -30 -30>
       angularTarget=<0 0 0>
       linearAxesMax=<10 20 20>
       linearAxesMin=<0 -20 -20>
      }
      localJointOffset=<0 0 0>
     }]
   }
   n32 = Anim Dynamics - Bone: dyn_r_arm_scrap_inner(ComponentPose=@n31, Alpha=0.06, ExternalForce="0,0,0") {
    angularDampingOverride=0.98
    angularSpringConstant=30
    bAngularSpring=true
    bGravityOverrideInSimSpace=true
    bLinearSpring=true
    bOverrideAngularDamping=true
    bOverrideLinearDamping=true
    gravityScale=0
    lODThreshold=1
    linearDampingOverride=0.98
    linearSpringConstant=30
    simulationSpace=Actor
    alphaBoolBlend {blendOption=Linear}
    boundBone {boneName=dyn_r_arm_scrap_inner}
    physicsBodyDefinitions [{
      sphereCollisionRadius=0
      boundBone {boneName=dyn_r_arm_scrap_inner}
      boxExtents=<30 30 30>
      constraintSetup {
       angularTargetAxis=AxisY
       angularLimitsMax=<0 30 30>
       angularLimitsMin=<0 -30 -30>
       angularTarget=<0 0 0>
       linearAxesMax=<10 20 20>
       linearAxesMin=<0 -20 -20>
      }
      localJointOffset=<0 0 0>
     }]
   }
   n33 = Anim Dynamics - Bone: dyn_r_arm_scrap_front(ComponentPose=@n32, Alpha=0.06, ExternalForce="0,0,0") {
    angularDampingOverride=0.98
    angularSpringConstant=30
    bAngularSpring=true
    bGravityOverrideInSimSpace=true
    bLinearSpring=true
    bOverrideAngularDamping=true
    bOverrideLinearDamping=true
    gravityScale=0
    lODThreshold=1
    linearDampingOverride=0.98
    linearSpringConstant=30
    simulationSpace=Actor
    alphaBoolBlend {blendOption=Linear}
    boundBone {boneName=dyn_r_arm_scrap_front}
    physicsBodyDefinitions [{
      sphereCollisionRadius=0
      boundBone {boneName=dyn_r_arm_scrap_front}
      boxExtents=<30 30 30>
      constraintSetup {
       angularTargetAxis=AxisY
       angularLimitsMax=<0 30 30>
       angularLimitsMin=<0 -30 -30>
       angularTarget=<0 0 0>
       linearAxesMax=<10 20 20>
       linearAxesMin=<0 -20 -20>
      }
      localJointOffset=<0 0 0>
     }]
   }
   n34 = Anim Dynamics - Bone: dyn_l_arm_scrap_front(ComponentPose=@n33, Alpha=0.06, ExternalForce="0,0,0") {
    angularDampingOverride=0.98
    angularSpringConstant=30
    bAngularSpring=true
    bGravityOverrideInSimSpace=true
    bLinearSpring=true
    bOverrideAngularDamping=true
    bOverrideLinearDamping=true
    gravityScale=0
    lODThreshold=1
    linearDampingOverride=0.98
    linearSpringConstant=30
    simulationSpace=Actor
    alphaBoolBlend {blendOption=Linear}
    boundBone {boneName=dyn_l_arm_scrap_front}
    physicsBodyDefinitions [{
      sphereCollisionRadius=0
      boundBone {boneName=dyn_l_arm_scrap_front}
      boxExtents=<30 30 30>
      constraintSetup {
       angularTargetAxis=AxisY
       angularLimitsMax=<0 30 30>
       angularLimitsMin=<0 -30 -30>
       angularTarget=<0 0 0>
       linearAxesMax=<10 20 20>
       linearAxesMin=<0 -20 -20>
      }
      localJointOffset=<0 0 0>
     }]
   }
   n35 = Anim Dynamics - Bone: dyn_l_arm_scrap_back(ComponentPose=@n34, Alpha=0.06, ExternalForce="0,0,0") {
    angularDampingOverride=0.98
    angularSpringConstant=30
    bAngularSpring=true
    bGravityOverrideInSimSpace=true
    bLinearSpring=true
    bOverrideAngularDamping=true
    bOverrideLinearDamping=true
    gravityScale=0
    lODThreshold=1
    linearDampingOverride=0.98
    linearSpringConstant=30
    simulationSpace=Actor
    alphaBoolBlend {blendOption=Linear}
    boundBone {boneName=dyn_l_arm_scrap_back}
    physicsBodyDefinitions [{
      sphereCollisionRadius=0
      boundBone {boneName=dyn_l_arm_scrap_back}
      boxExtents=<30 30 30>
      constraintSetup {
       angularTargetAxis=AxisY
       angularLimitsMax=<0 30 30>
       angularLimitsMin=<0 -30 -30>
       angularTarget=<0 0 0>
       linearAxesMax=<10 20 20>
       linearAxesMin=<0 -20 -20>
      }
      localJointOffset=<0 0 0>
     }]
   }
   n36 = RigidBody(ComponentPose=@n35) {
    bOverrideWorldGravity=true
    overridePhysicsAsset=&/M_MED_Werewolf_01_Body_Physics
    simulationSpace=BaseBoneSpace
    alphaBoolBlend {blendOption=Linear}
    baseBoneRef {boneName=spine_03}
    componentLinearAccScale=<1.5 1.5 1.5>
    componentLinearVelScale=<1.5 1.5 1.5>
    overrideWorldGravity {z=-980}
   }
   Output Pose(Result=Component To Local(ComponentPose=@n36)) {layerGroup=None}
  ]}
 EventGraph {nodes [
   TryGetPawnOwner()
   Event BlueprintUpdateAnimation()
  ]}
]
variables []
