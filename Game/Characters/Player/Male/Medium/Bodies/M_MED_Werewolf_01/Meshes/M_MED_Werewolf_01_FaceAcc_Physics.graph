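# Physics asset for the medium male Werewolf 01 player's face accessory, setting up secondary motion for a two-tailed bandanna.
# Left and right chains of five box bodies hang from the head, each segment constrained to the one before it.
# Twist stays tight throughout, while one swing axis narrows and the other widens down each chain.
# Head, upper spine, clavicle and upper-arm spheres or capsules form unconstrained body volumes for the cloth tails to collide against.
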PhysicsAsset /Game/Characters/Player/Male/Medium/Bodies/M_MED_Werewolf_01/Meshes/M_MED_Werewolf_01_FaceAcc_Physics
bodies {
 clavicle_l [{shapeType=Capsule, length=8.75, radius=9.75, center=<9.7252 0 0>, rotation=<rot -90 0.0003 -0.0002>}]
 clavicle_r [{shapeType=Capsule, length=8.75, radius=9.75, center=<-9.4582 0.0001 0>, rotation=<rot -90 0.0003 -0.0002>}]
 dyn_bandanna_l_01 [{shapeType=Box, extentX=2.2721, extentY=7.25, extentZ=2.5183, center=<1.7191 0 0>}]
 dyn_bandanna_l_02 [{shapeType=Box, extentX=4.3201, extentY=6.76, extentZ=2.2889, center=<3.1271 0 0.0001>}]
 dyn_bandanna_l_03 [{shapeType=Box, extentX=8.029, extentY=7.0603, extentZ=3.6961, center=<5.1139 0 0>}]
 dyn_bandanna_l_04 [{shapeType=Box, extentX=8.1159, extentY=7.9566, extentZ=3.8852, center=<5.2706 0 -0.1827>}]
 dyn_bandanna_l_05 [{shapeType=Box, extentX=8.0365, extentY=8.2935, extentZ=4.3674, center=<5.292 0 -0.1827>}]
 dyn_bandanna_r_01 [{shapeType=Box, extentX=2.2721, extentY=7.25, extentZ=2.5183, center=<2.2787 0 0>}]
 dyn_bandanna_r_02 [{shapeType=Box, extentX=3.3645, extentY=6.76, extentZ=2.2889, center=<2.7394 0 0>}]
 dyn_bandanna_r_03 [{shapeType=Box, extentX=6.9691, extentY=7.0603, extentZ=2.4461, center=<4.7755 0 0>}]
 dyn_bandanna_r_04 [{shapeType=Box, extentX=9.8705, extentY=7.9566, extentZ=3.6352, center=<6.0697 -0.0001 -0.1827>}]
 dyn_bandanna_r_05 [{shapeType=Box, extentX=8.0365, extentY=8.2935, extentZ=3.6174, center=<5.2911 -0.0001 -0.1827>}]
 head [{shapeType=Sphere, radius=41.5, center=<43.1972 12.1008 0.0002>}, {shapeType=Capsule, length=0, radius=12.25, center=<6.5747 5.237 -0.0001>}]
 spine_04 [{shapeType=Capsule, length=0, radius=24.7315, center=<-5.3579 13.3329 0.3113>, rotation=<rot 85.3963 -115.3918 -44.9908>}]
 spine_05 [{shapeType=Capsule, length=0, radius=27.2315, center=<-5.3579 13.3329 0.3113>, rotation=<rot 85.3963 -115.3918 -44.9908>}]
 upperarm_l [{shapeType=Capsule, length=26.75, radius=9.25, center=<13.1604 0.6274 -0.2085>, rotation=<rot -88.1986 -86.898 89.6603>}]
 upperarm_r [{shapeType=Capsule, length=26.75, radius=9.25, center=<-12.903 -0.6028 0.6643>, rotation=<rot -88.1987 -86.8981 89.6604>}]
}
constraints [
 {bone1Name=dyn_bandanna_l_01, bone2Name=head, swing1LimitDegrees=15, swing2LimitDegrees=90, twistLimitDegrees=3}
 {bone1Name=dyn_bandanna_r_01, bone2Name=head, swing1LimitDegrees=15, swing2LimitDegrees=90, twistLimitDegrees=3}
 {bone1Name=dyn_bandanna_r_02, bone2Name=dyn_bandanna_r_01, swing1LimitDegrees=15, swing2LimitDegrees=45, twistLimitDegrees=3}
 {bone1Name=dyn_bandanna_l_02, bone2Name=dyn_bandanna_l_01, swing1LimitDegrees=15, swing2LimitDegrees=45, twistLimitDegrees=3}
 {bone1Name=dyn_bandanna_l_03, bone2Name=dyn_bandanna_l_02, swing1LimitDegrees=30, swing2LimitDegrees=45, twistLimitDegrees=3}
 {bone1Name=dyn_bandanna_r_03, bone2Name=dyn_bandanna_r_02, swing1LimitDegrees=30, swing2LimitDegrees=45, twistLimitDegrees=3}
 {bone1Name=dyn_bandanna_r_04, bone2Name=dyn_bandanna_r_03, swing1LimitDegrees=30, swing2LimitDegrees=30, twistLimitDegrees=3}
 {bone1Name=dyn_bandanna_r_05, bone2Name=dyn_bandanna_r_04, swing1LimitDegrees=30, swing2LimitDegrees=30, twistLimitDegrees=3}
 {bone1Name=dyn_bandanna_l_04, bone2Name=dyn_bandanna_l_03, swing1LimitDegrees=30, swing2LimitDegrees=30, twistLimitDegrees=3}
 {bone1Name=dyn_bandanna_l_05, bone2Name=dyn_bandanna_l_04, swing1LimitDegrees=30, swing2LimitDegrees=30, twistLimitDegrees=3}
]
