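# Post-attach animation blueprint for the Werewolf face accessory mesh, copying pose and curves from its parent mesh and adding secondary motion.
# The main branch chains spring Anim Dynamics on six hair bones, a head-space RigidBody node, then lightly weighted ear dynamics.
# When rigid body is disabled or LOD exceeds 1, a cheaper branch instead aims bandanna bones at spine bones with Look At nodes.
# The bake-pose layered blend is unconnected to output.
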AnimBlueprint /Game/Characters/Player/Male/Medium/Bodies/M_MED_Werewolf_01/Meshes/M_MED_Werewolf_01_FaceAcc_AnimBP
target_skeleton=&/M_MED_Werewolf_01_FaceAcc_Skeleton
graphs [
 AnimGraph {nodes [
   n1 = Sequence Player 'M_MED_Werewolf_01_FaceAcc_BakePose'() {bLoopAnimation=false, sequence=&/M_MED_Werewolf_01_FaceAcc_BakePose}
   Layered blend per bone(BlendPoses_0=@n1) {layerSetup [{branchFilters [{blendDepth=1, boneName=dyn_bandanna_l_mstr}, {blendDepth=1, boneName=dyn_bandanna_r_mstr}]}]}
   Local Space Ref Pose()
   n2 = Anim Dynamics - Bone: dyn_hair_01(ComponentPose=Local To Component(LocalPose=Use cached pose 'CPFM Cache'()), Alpha=1) {
    angularDampingOverride=0.98
    angularSpringConstant=90
    bAngularSpring=true
    bGravityOverrideInSimSpace=true
    bLinearSpring=true
    bOverrideLinearDamping=true
    gravityScale=0
    lODThreshold=1
    linearDampingOverride=0.98
    linearSpringConstant=100
    simulationSpace=Actor
    alphaBoolBlend {blendOption=Linear}
    alphaScaleBiasClamp {
     bClampResult=true
     bInterpResult=true
     clampMax=0.25
     clampMin=0.05
     interpSpeedDecreasing=0.5
     interpSpeedIncreasing=0.5
     outRange {max=0.1}
    }
    boundBone {boneName=dyn_hair_01}
    physicsBodyDefinitions [{
      sphereCollisionRadius=0
      boundBone {boneName=dyn_hair_01}
      boxExtents=<30 30 30>
      constraintSetup {
       angularTargetAxis=AxisY
       twistAxis=AxisZ
       angularLimitsMax=<30 2 0>
       angularLimitsMin=<-30 -2 0>
       angularTarget=<0 0 0>
       linearAxesMax=<0 0 5>
       linearAxesMin=<0 0 -1>
      }
      localJointOffset=<0 0 0>
     }]
   }
   n3 = Anim Dynamics - Bone: dyn_hair_02(ComponentPose=@n2, Alpha=0.125) {
    angularDampingOverride=0.98
    angularSpringConstant=90
    bAngularSpring=true
    bGravityOverrideInSimSpace=true
    bLinearSpring=true
    bOverrideAngularDamping=true
    bOverrideLinearDamping=true
    gravityScale=0.125
    lODThreshold=1
    linearDampingOverride=0.98
    linearSpringConstant=100
    simulationSpace=Actor
    alphaBoolBlend {blendOption=Linear}
    alphaScaleBiasClamp {
     bClampResult=true
     bInterpResult=true
     clampMax=0.25
     clampMin=0.05
     interpSpeedDecreasing=0.5
     interpSpeedIncreasing=0.5
     outRange {max=0.1}
    }
    boundBone {boneName=dyn_hair_02}
    physicsBodyDefinitions [{
      sphereCollisionRadius=0
      boundBone {boneName=dyn_hair_02}
      boxExtents=<30 30 30>
      constraintSetup {
       angularTargetAxis=AxisY
       twistAxis=AxisZ
       angularLimitsMax=<2.5 30 0>
       angularLimitsMin=<-2.5 -30 0>
       angularTarget=<0 0 0>
       linearAxesMax=<10 2.5 5>
       linearAxesMin=<-10 -2.5 0>
      }
      localJointOffset=<0 0 -5>
     }]
   }
   n4 = Anim Dynamics - Bone: dyn_hair_03(ComponentPose=@n3, Alpha=0.75, ExternalForce="0,0,0") {
    angularDampingOverride=0.98
    angularSpringConstant=90
    bAngularSpring=true
    bGravityOverrideInSimSpace=true
    bLinearSpring=true
    bOverrideLinearDamping=true
    gravityScale=0.1
    lODThreshold=1
    linearDampingOverride=0.98
    linearSpringConstant=100
    simulationSpace=Actor
    alphaBoolBlend {blendOption=Linear}
    alphaScaleBiasClamp {
     bClampResult=true
     bInterpResult=true
     clampMax=0.25
     clampMin=0.05
     interpSpeedDecreasing=0.5
     interpSpeedIncreasing=0.5
     outRange {max=0.1}
    }
    boundBone {boneName=dyn_hair_03}
    physicsBodyDefinitions [{
      sphereCollisionRadius=0
      boundBone {boneName=dyn_hair_03}
      boxExtents=<30 30 30>
      constraintSetup {
       angularTargetAxis=AxisY
       twistAxis=AxisZ
       angularLimitsMax=<30 2 0>
       angularLimitsMin=<-30 -2 0>
       angularTarget=<0 0 0>
       linearAxesMax=<0 0 5>
       linearAxesMin=<0 0 -1>
      }
      localJointOffset=<0 0 0>
     }]
   }
   n5 = Anim Dynamics - Bone: dyn_hair_04(ComponentPose=@n4, Alpha=0.3, ExternalForce="0,0,0") {
    angularDampingOverride=0.98
    angularSpringConstant=90
    bAngularSpring=true
    bGravityOverrideInSimSpace=true
    bLinearSpring=true
    bOverrideAngularDamping=true
    bOverrideLinearDamping=true
    gravityScale=0.2
    lODThreshold=1
    linearDampingOverride=0.98
    linearSpringConstant=100
    simulationSpace=Actor
    alphaBoolBlend {blendOption=Linear}
    alphaScaleBiasClamp {
     bClampResult=true
     bInterpResult=true
     clampMax=0.25
     clampMin=0.05
     interpSpeedDecreasing=0.5
     interpSpeedIncreasing=0.5
     outRange {max=0.1}
    }
    boundBone {boneName=dyn_hair_04}
    physicsBodyDefinitions [{
      sphereCollisionRadius=0
      boundBone {boneName=dyn_hair_04}
      boxExtents=<30 30 30>
      constraintSetup {
       angularTargetAxis=AxisY
       twistAxis=AxisZ
       angularLimitsMax=<2.5 30 0>
       angularLimitsMin=<-2.5 -30 0>
       angularTarget=<0 0 0>
       linearAxesMax=<10 2.5 5>
       linearAxesMin=<-10 -2.5 0>
      }
      localJointOffset=<0 0 -5>
     }]
   }
   n6 = Anim Dynamics - Bone: dyn_hair_05(ComponentPose=@n5, Alpha=0.85, ExternalForce="0,0,0") {
    angularDampingOverride=0.98
    angularSpringConstant=90
    bAngularSpring=true
    bGravityOverrideInSimSpace=true
    bLinearSpring=true
    bOverrideLinearDamping=true
    gravityScale=0.05
    lODThreshold=1
    linearDampingOverride=0.98
    linearSpringConstant=100
    simulationSpace=Actor
    alphaBoolBlend {blendOption=Linear}
    alphaScaleBiasClamp {
     bClampResult=true
     bInterpResult=true
     clampMax=0.25
     clampMin=0.05
     interpSpeedDecreasing=0.5
     interpSpeedIncreasing=0.5
     outRange {max=0.1}
    }
    boundBone {boneName=dyn_hair_05}
    physicsBodyDefinitions [{
      sphereCollisionRadius=0
      boundBone {boneName=dyn_hair_05}
      boxExtents=<30 30 30>
      constraintSetup {
       angularTargetAxis=AxisY
       twistAxis=AxisZ
       angularLimitsMax=<30 2 0>
       angularLimitsMin=<-30 -2 0>
       angularTarget=<0 0 0>
       linearAxesMax=<2 2 5>
       linearAxesMin=<-2 -2 -2>
      }
      localJointOffset=<0 0 0>
     }]
   }
   n7 = Anim Dynamics - Bone: dyn_hair_06(ComponentPose=@n6, Alpha=0.25, ExternalForce="0,0,0") {
    angularDampingOverride=0.98
    angularSpringConstant=90
    bAngularSpring=true
    bGravityOverrideInSimSpace=true
    bLinearSpring=true
    bOverrideAngularDamping=true
    bOverrideLinearDamping=true
    gravityScale=0.175
    lODThreshold=1
    linearDampingOverride=0.98
    linearSpringConstant=100
    simulationSpace=Actor
    alphaBoolBlend {blendOption=Linear}
    alphaScaleBiasClamp {
     bClampResult=true
     bInterpResult=true
     clampMax=0.25
     clampMin=0.05
     interpSpeedDecreasing=0.5
     interpSpeedIncreasing=0.5
     outRange {max=0.1}
    }
    boundBone {boneName=dyn_hair_06}
    physicsBodyDefinitions [{
      sphereCollisionRadius=0
      boundBone {boneName=dyn_hair_06}
      boxExtents=<30 30 30>
      constraintSetup {
       angularTargetAxis=AxisY
       twistAxis=AxisZ
       angularLimitsMax=<2.5 30 0>
       angularLimitsMin=<-2.5 -30 0>
       angularTarget=<0 0 0>
       linearAxesMax=<10 2.5 5>
       linearAxesMin=<-10 -2.5 0>
      }
      localJointOffset=<0 0 -5>
     }]
   }
   n8 = RigidBody(ComponentPose=@n7, Alpha=0.99) {
    bOverrideWorldGravity=true
    overridePhysicsAsset=&/M_MED_Werewolf_01_FaceAcc_Physics
    simulationSpace=BaseBoneSpace
    alphaBoolBlend {blendOption=Linear}
    baseBoneRef {boneName=head}
    componentAppliedLinearAccClamp=<5000 5000 5000>
    componentLinearAccScale=<0.45 0.45 0.45>
    componentLinearVelScale=<0.45 0.45 0.45>
    overrideWorldGravity {z=-1000}
   }
   n9 = Anim Dynamics - Bone: dyn_r_ear(ComponentPose=@n8, Alpha=0.05, ExternalForce="0,0,0") {
    angularDampingOverride=0.97
    angularSpringConstant=80
    bAngularSpring=true
    bGravityOverrideInSimSpace=true
    bLinearSpring=true
    bOverrideAngularDamping=true
    bOverrideLinearDamping=true
    gravityScale=0.1
    lODThreshold=1
    linearDampingOverride=0.97
    linearSpringConstant=80
    simulationSpace=Actor
    alphaBoolBlend {blendOption=Linear}
    boundBone {boneName=dyn_r_ear}
    physicsBodyDefinitions [{
      sphereCollisionRadius=0
      boundBone {boneName=dyn_r_ear}
      boxExtents=<30 30 30>
      constraintSetup {
       angularTargetAxis=AxisY
       angularLimitsMax=<0 30 30>
       angularLimitsMin=<0 -30 -30>
       angularTarget=<0 0 0>
       linearAxesMax=<10 20 20>
       linearAxesMin=<0 -5 -5>
      }
      localJointOffset=<0 -0.5 0>
     }]
   }
   n10 = Anim Dynamics - Bone: dyn_l_ear(ComponentPose=@n9, Alpha=0.04, ExternalForce="0,0,0") {
    angularSpringConstant=80
    bAngularSpring=true
    bGravityOverrideInSimSpace=true
    bLinearSpring=true
    gravityScale=0.1
    lODThreshold=1
    linearSpringConstant=80
    simulationSpace=Actor
    alphaBoolBlend {blendOption=Linear}
    boundBone {boneName=dyn_l_ear}
    physicsBodyDefinitions [{
      sphereCollisionRadius=0
      boundBone {boneName=dyn_l_ear}
      boxExtents=<30 30 30>
      constraintSetup {
       angularTargetAxis=AxisY
       angularLimitsMax=<0 30 30>
       angularLimitsMin=<0 -30 -30>
       angularTarget=<0 0 0>
       linearAxesMax=<10 20 5>
       linearAxesMin=<0 0 -20>
      }
      localJointOffset=<0 0 0>
     }]
   }
   n11 = Look At - Bone: dyn_bandanna_l_01(ComponentPose=Local To Component(LocalPose=Use cached pose 'CPFM Cache'()), Alpha=1) {
    alphaBoolBlend {blendOption=Linear}
    boneToModify {boneName=dyn_bandanna_l_01}
    lookAtLocation=<0 -20 -10>
    lookAtTarget {boneReference {boneName=spine_05}}
    lookAt_Axis {axis=<1 0>}
   }
   n12 = Look At - Bone: dyn_bandanna_r_01(ComponentPose=@n11, Alpha=1) {
    alphaBoolBlend {blendOption=Linear}
    boneToModify {boneName=dyn_bandanna_r_01}
    lookAtLocation=<0 -20 10>
    lookAtTarget {boneReference {boneName=spine_05}}
    lookAt_Axis {axis=<1 0>}
   }
   n13 = Look At - Bone: dyn_bandanna_l_04(ComponentPose=@n12, Alpha=1) {
    alphaBoolBlend {blendOption=Linear}
    boneToModify {boneName=dyn_bandanna_l_04}
    lookAtLocation=<0 -20 -10>
    lookAtTarget {boneReference {boneName=spine_04}}
    lookAt_Axis {axis=<1 0>}
   }
   n14 = Look At - Bone: dyn_bandanna_r_04(ComponentPose=@n13, Alpha=1) {
    alphaBoolBlend {blendOption=Linear}
    boneToModify {boneName=dyn_bandanna_r_04}
    lookAtLocation=<0 -20 10>
    lookAtTarget {boneReference {boneName=spine_04}}
    lookAt_Axis {axis=<1 0>}
   }
   n15 = "integer > integer"(A=Get CurrentLOD(), B=1) {is=title K2Node_CallFunction}
   n16 = NOT Boolean(A=Get bRigidBodyEnabled())
   Output Pose(Result=Blend Poses by bool(
     BlendPose_0=Component To Local(ComponentPose=@n14)
     BlendPose_1=Component To Local(ComponentPose=@n10)
     bActiveValue=OR Boolean(A=@n16, B=@n15)
    ) {blendType=Linear}) {layerGroup=None}
   Save cached pose 'CPFM Cache'(Pose=Copy Pose From Mesh() {bCopyCurves=true, bUseAttachedParent=true}) {cacheName=CPFM Cache}
  ]}
 EventGraph {nodes [
   TryGetPawnOwner()
   Event BlueprintUpdateAnimation()
  ]}
]
variables []
