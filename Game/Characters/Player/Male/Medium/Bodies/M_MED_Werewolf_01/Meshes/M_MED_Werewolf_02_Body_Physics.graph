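# Physics asset for the medium male Werewolf 02 player body, covering a two-part nunchuck rig of dynamic bones plus the pelvis and left thigh.
# Capsule bodies form upper and lower nunchuck chains, each constrained to the pelvis at its 02 bone with narrow twist and swing limits.
# The left thigh joins the pelvis with free rotation, while both tiny 01 root capsules have no constraints.
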PhysicsAsset /Game/Characters/Player/Male/Medium/Bodies/M_MED_Werewolf_01/Meshes/M_MED_Werewolf_02_Body_Physics
bodies {
 dyn_lower_nunchuck_01 [{shapeType=Capsule, length=0, radius=0.5}]
 dyn_lower_nunchuck_02 [{shapeType=Capsule, length=2.25, radius=1.5, center=<2.8044 0 0.114>, rotation=<rot -87.6717 0.0001 0>}]
 dyn_lower_nunchuck_03 [{shapeType=Capsule, length=1.6531, radius=1.5794, center=<2.5236 0 0>, rotation=<rot 90 0 0>}]
 dyn_lower_nunchuck_04 [{shapeType=Capsule, length=19.744, radius=1.5, center=<11.6565 -0.0001 -0.6207>, rotation=<rot 84.9999 -0.0001 0>}]
 dyn_upper_nunchuck_01 [{shapeType=Capsule, length=0, radius=0.5}]
 dyn_upper_nunchuck_02 [{shapeType=Capsule, length=1, radius=2, center=<2.0113 0.0001 0.2342>, rotation=<rot -83.0133 0 0>}]
 dyn_upper_nunchuck_03 [{shapeType=Capsule, length=15, radius=1.5, center=<12.6758 -0.0702 -5.86>, rotation=<rot 64.2334 -1.5639 -2.6611>}]
 pelvis [{shapeType=Capsule, length=0, radius=13, center=<0 2.8244 0>}]
 wolf_thigh_l [{shapeType=Capsule, length=20, radius=8, center=<-12.1448 1.3382 1.1835>, rotation=<rot -84.8495 -19.7019 9.779>}]
}
constraints [
 {
  bone1Name=wolf_thigh_l
  bone2Name=pelvis
  swing1LimitDegrees=45
  swing1Motion=Free
  swing2LimitDegrees=45
  swing2Motion=Free
  twistLimitDegrees=45
  twistMotion=Free
 }
 {bone1Name=dyn_upper_nunchuck_02, bone2Name=pelvis, swing1LimitDegrees=30, swing2LimitDegrees=30, twistLimitDegrees=2}
 {bone1Name=dyn_upper_nunchuck_03, bone2Name=dyn_upper_nunchuck_02, swing1LimitDegrees=30, swing2LimitDegrees=30, twistLimitDegrees=3}
 {bone1Name=dyn_lower_nunchuck_04, bone2Name=dyn_lower_nunchuck_03, swing1LimitDegrees=20, swing2LimitDegrees=20, twistLimitDegrees=2}
 {bone1Name=dyn_lower_nunchuck_02, bone2Name=pelvis, swing1LimitDegrees=6, swing2LimitDegrees=6, twistLimitDegrees=1}
 {bone1Name=dyn_lower_nunchuck_03, bone2Name=dyn_lower_nunchuck_02, swing1LimitDegrees=10, swing2LimitDegrees=10, twistLimitDegrees=1}
]
