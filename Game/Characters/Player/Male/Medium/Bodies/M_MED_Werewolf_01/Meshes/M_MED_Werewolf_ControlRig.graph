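# Control Rig for the player werewolf body that retargets human leg animation onto its separate wolf leg bones.
# Per leg, thigh, calf and foot local rotations, corrected by rest-pose offsets, drive wolf thigh, calf and ball.
# FABRIK then pulls the wolf chain to a toe target derived from the human ball bone, and the toe snaps to it.
# Foot offsets are asymmetric: left references wolf_ball_l, right wolf_foot_r.
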ControlRigBlueprint /Game/Characters/Player/Male/Medium/Bodies/M_MED_Werewolf_01/Meshes/M_MED_Werewolf_ControlRig
graphs [
 RigVMFunctionLibrary {links [], nodes []}
 RigVMModel {nodes [
   BeginExecution_16_0 = &RigUnit_BeginExecution()
   GetBoneTransform = &RigUnit_GetBoneTransform(Bone=thigh_l, Space=LocalSpace)
   GetInitialBoneTransform = /Script/ControlRig.RigUnit_GetInitialBoneTransform(Bone=wolf_thigh_l, Space=GlobalSpace) {is=struct RigVMUnitNode}
   GetInitialBoneTransform_0 = &RigUnit_GetInitialBoneTransform(Bone=thigh_l, Space=GlobalSpace)
   MathTransformMakeRelative = &RigVMFunction_MathTransformMakeRelative(Parent=@GetInitialBoneTransform.Transform, Global=@GetInitialBoneTransform_0.Transform)
   MathTransformMakeRelative_1 = &RigVMFunction_MathTransformMakeRelative(Global=@GetBoneTransform.Transform, Parent=@MathTransformMakeRelative.Local)
   SetBoneRotation = /Script/ControlRig.RigUnit_SetBoneRotation(
    Bone=wolf_thigh_l
    Space=LocalSpace
    Weight=1
    bPropagateToChildren=True
    ExecutePin=@BeginExecution_16_0.ExecutePin
    Rotation=@MathTransformMakeRelative_1.Local.Rotation
   ) {is=struct RigVMUnitNode}
   n1 = &RigUnit_GetInitialBoneTransform(Bone=calf_l, Space=GlobalSpace)
   GetInitialBoneTransform_2 = &RigUnit_GetInitialBoneTransform(Bone=wolf_calf_l, Space=GlobalSpace)
   > &RigUnit_SetBoneRotation(
    Bone=wolf_calf_l
    Space=LocalSpace
    Weight=1
    bPropagateToChildren=True
    Rotation=&RigVMFunction_MathTransformMakeRelative(
     Global=&RigUnit_GetBoneTransform(Bone=calf_l, Space=LocalSpace).Transform
     Parent=&RigVMFunction_MathTransformMakeRelative(Global=@n1.Transform, Parent=@GetInitialBoneTransform_2.Transform).Local
    ).Local.Rotation
   )
   n2 = &RigUnit_GetInitialBoneTransform(Bone=foot_l, Space=GlobalSpace)
   n3 = &RigUnit_GetInitialBoneTransform(Bone=wolf_ball_l, Space=GlobalSpace)
   > &RigUnit_SetBoneRotation(
    Bone=wolf_ball_l
    Space=LocalSpace
    Weight=1
    bPropagateToChildren=True
    Rotation=&RigVMFunction_MathTransformMakeRelative(
     Global=&RigUnit_GetBoneTransform(Bone=foot_l, Space=LocalSpace).Transform
     Parent=&RigVMFunction_MathTransformMakeRelative(Global=@n2.Transform, Parent=@n3.Transform).Local
    ).Local.Rotation
   )
   n4 = &RigVMFunction_MathTransformMul(
    B=&RigUnit_GetBoneTransform(Bone=ball_l).Transform
    A=&RigVMFunction_MathTransformMakeRelative(
     Parent=&RigUnit_GetInitialBoneTransform(Bone=ball_l, Space=GlobalSpace).Transform
     Global=&RigUnit_GetInitialBoneTransform(Bone=wolf_toe_l, Space=GlobalSpace).Transform
    ).Local
   ) {template="Multiply::Execute(in A,in B,out Result)"}
   > /Script/ControlRig.RigUnit_FABRIK(
    EffectorBone=wolf_toe_l
    MaxIterations=10
    Precision=1
    StartBone=wolf_thigh_l
    Weight=1
    bPropagateToChildren=True
    bSetEffectorTransform=True
    EffectorTransform=@n4.Result
   ) {is=struct RigVMUnitNode}
   > /Script/ControlRig.RigUnit_SetBoneTransform(Bone=wolf_toe_l, Space=GlobalSpace, Weight=1, bPropagateToChildren=True, Transform=@n4.Result) {is=struct RigVMUnitNode}
   BeginExecution_16_0_6 = &RigUnit_BeginExecution()
   GetBoneTransform_2 = &RigUnit_GetBoneTransform(Bone=thigh_r, Space=LocalSpace)
   GetInitialBoneTransform_0_0 = &RigUnit_GetInitialBoneTransform(Bone=thigh_r, Space=GlobalSpace)
   GetInitialBoneTransform_3 = &RigUnit_GetInitialBoneTransform(Bone=wolf_thigh_r, Space=GlobalSpace)
   MathTransformMakeRelative_2 = &RigVMFunction_MathTransformMakeRelative(Global=@GetInitialBoneTransform_0_0.Transform, Parent=@GetInitialBoneTransform_3.Transform)
   MathTransformMakeRelative_1_2 = &RigVMFunction_MathTransformMakeRelative(Global=@GetBoneTransform_2.Transform, Parent=@MathTransformMakeRelative_2.Local)
   SetBoneRotation_1 = &RigUnit_SetBoneRotation(
    Bone=wolf_thigh_r
    Space=LocalSpace
    Weight=1
    bPropagateToChildren=True
    ExecutePin=@BeginExecution_16_0_6.ExecutePin
    Rotation=@MathTransformMakeRelative_1_2.Local.Rotation
   )
   n5 = &RigUnit_GetInitialBoneTransform(Bone=calf_r, Space=GlobalSpace)
   GetInitialBoneTransform_2_1 = &RigUnit_GetInitialBoneTransform(Bone=wolf_calf_r, Space=GlobalSpace)
   > &RigUnit_SetBoneRotation(
    Bone=wolf_calf_r
    Space=LocalSpace
    Weight=1
    bPropagateToChildren=True
    Rotation=&RigVMFunction_MathTransformMakeRelative(
     Global=&RigUnit_GetBoneTransform(Bone=calf_r, Space=LocalSpace).Transform
     Parent=&RigVMFunction_MathTransformMakeRelative(Global=@n5.Transform, Parent=@GetInitialBoneTransform_2_1.Transform).Local
    ).Local.Rotation
   )
   n6 = &RigUnit_GetInitialBoneTransform(Bone=foot_r, Space=GlobalSpace)
   n7 = &RigUnit_GetInitialBoneTransform(Bone=wolf_foot_r, Space=GlobalSpace)
   > &RigUnit_SetBoneRotation(
    Bone=wolf_ball_r
    Space=LocalSpace
    Weight=1
    bPropagateToChildren=True
    Rotation=&RigVMFunction_MathTransformMakeRelative(
     Global=&RigUnit_GetBoneTransform(Bone=foot_r, Space=LocalSpace).Transform
     Parent=&RigVMFunction_MathTransformMakeRelative(Global=@n6.Transform, Parent=@n7.Transform).Local
    ).Local.Rotation
   )
   n8 = &RigVMFunction_MathTransformMul(
    B=&RigUnit_GetBoneTransform(Bone=ball_r).Transform
    A=&RigVMFunction_MathTransformMakeRelative(
     Parent=&RigUnit_GetInitialBoneTransform(Bone=ball_r, Space=GlobalSpace).Transform
     Global=&RigUnit_GetInitialBoneTransform(Bone=wolf_toe_r, Space=GlobalSpace).Transform
    ).Local
   ) {template="Multiply::Execute(in A,in B,out Result)"}
   > &RigUnit_FABRIK(
    EffectorBone=wolf_toe_r
    MaxIterations=10
    Precision=1
    StartBone=wolf_thigh_r
    Weight=1
    bPropagateToChildren=True
    bSetEffectorTransform=True
    EffectorTransform=@n8.Result
   )
   > &RigUnit_SetBoneTransform(Bone=wolf_toe_r, Space=GlobalSpace, Weight=1, bPropagateToChildren=True, Transform=@n8.Result)
  ]}
]
hierarchy [
 root {loc=<0 0 0>, rot=<0 0 0 1>}
 attach {parent=root, loc=<0 0 0>, rot=<0 0 0 1>}
 pelvis {parent=root, loc=<0 -0.7071 93.1191>, rot=<0.0127 0.707 -0.0127 -0.707>}
 spine_01 {parent=pelvis, loc=<6.7494 -0.6388 0>, rot=<0 0 0.034 -0.9994>}
 spine_02 {parent=spine_01, loc=<7.2518 -0.035 0>, rot=<0 0 0.0238 -0.9997>}
 spine_03 {parent=spine_02, loc=<6.0104 0 0>, rot=<0 0 0.0149 -0.9999>}
 spine_04 {parent=spine_03, loc=<8.1024 0 0>, rot=<0 0 -0.0132 -0.9999>}
 spine_05 {parent=spine_04, loc=<10.0146 0 0>, rot=<0 0 -0.0015 -1>}
 clavicle_l {parent=spine_05, loc=<4.3575 -0.1466 -2.3223>, rot=<0.0465 -0.7289 -0.009 -0.6829>}
 upperarm_l {parent=clavicle_l, loc=<15.5111 0 0>, rot=<0.0305 -0.536 -0.0057 -0.8437>}
 lowerarm_l {parent=upperarm_l, loc=<26.6514 0 0>, rot=<0.0391 -0.0099 -0.2453 -0.9686>}
 hand_l {parent=lowerarm_l, loc=<22.3491 0 0>, rot=<-0.7081 -0.0507 0.0975 -0.6975>}
 index_metacarpal_l {parent=hand_l, loc=<1.6535 0.0183 -2.96>, rot=<0.0158 -0.0124 -0.0173 -0.9996>}
 index_01_l {parent=index_metacarpal_l, loc=<8.3008 0.0357 -0.0524>, rot=<-0.0044 0.0485 0.0924 -0.9945>}
 index_02_l {parent=index_01_l, loc=<3.3587 -0.3003 0.0452>, rot=<-0.002 0.0027 0.3225 -0.9466>}
 index_03_l {parent=index_02_l, loc=<2.5577 -0.1233 -0.0252>, rot=<-0.001 0.0068 0.2558 -0.9667>}
 middle_metacarpal_l {parent=hand_l, loc=<2.1856 -0.1725 -0.5984>, rot=<0.006 0.0168 -0.0329 -0.9993>}
 middle_01_l {parent=middle_metacarpal_l, loc=<7.3275 -0.2447 0.0129>, rot=<-0.0208 0.0042 0.0858 -0.9961>}
 middle_02_l {parent=middle_01_l, loc=<4.0131 0.0722 0.0029>, rot=<0 0 0.3511 -0.9363>}
 middle_03_l {parent=middle_02_l, loc=<2.6079 0.1152 0.0049>, rot=<0 0 0.1707 -0.9853>}
 pinky_metacarpal_l {parent=hand_l, loc=<1.7616 -0.9753 3.3814>, rot=<-0.0397 0.0814 -0.069 -0.9935>}
 pinky_01_l {parent=pinky_metacarpal_l, loc=<7.1709 -0.4831 -0.1094>, rot=<-0.0654 -0.0035 0.2242 -0.9723>}
 pinky_02_l {parent=pinky_01_l, loc=<2.8274 0.0893 -0.1043>, rot=<0.0037 0.0011 0.2854 -0.9584>}
 pinky_03_l {parent=pinky_02_l, loc=<2.2046 0.0307 -0.0078>, rot=<-0.0002 0.0098 0.1445 -0.9895>}
 ring_metacarpal_l {parent=hand_l, loc=<2.2074 -0.3625 1.3638>, rot=<-0.0284 0.0722 -0.0458 -0.9959>}
 ring_01_l {parent=ring_metacarpal_l, loc=<7.0295 -0.6197 -0.1446>, rot=<-0.0489 -0.0252 0.1892 -0.9804>}
 ring_02_l {parent=ring_01_l, loc=<3.6174 0.4647 -0.1027>, rot=<0 0 0.2806 -0.9598>}
 ring_03_l {parent=ring_02_l, loc=<2.4152 0.114 0.0199>, rot=<0 0 0.1189 -0.9929>}
 thumb_01_l {parent=hand_l, loc=<1.3118 -0.3191 -2.459>, rot=<0.6546 -0.4042 0.0664 -0.6354>}
 thumb_02_l {parent=thumb_01_l, loc=<5.9337 -0.0448 0.0401>, rot=<-0.0037 0.0213 0.298 -0.9543>}
 thumb_03_l {parent=thumb_02_l, loc=<2.2223 0.3682 -0.0907>, rot=<0.0118 -0.013 0.0066 -0.9998>}
 weapon_l {parent=hand_l, loc=<-0.1974 -3.4111 0.3811>, rot=<0 0 0 -1>}
 lowerarm_twist_01_l {parent=lowerarm_l, loc=<17.8549 0 0>, rot=<0 0 0 -1>}
 lowerarm_twist_02_l {parent=lowerarm_l, loc=<9.6199 0 0>, rot=<0 0 0 -1>}
 upperarm_twist_01_l {parent=upperarm_l, loc=<0.0041 0 0>, rot=<0 0 0 1>}
 upperarm_twist_02_l {parent=upperarm_l, loc=<13.7941 0 0>, rot=<0 0 0 1>}
 dyn_shoulderpad_l {parent=clavicle_l, loc=<7.1878 -2.4274 11.5866>, rot=<0.9991 0.0252 0.0336 0.0008>}
 clavicle_r {parent=spine_05, loc=<4.3575 -0.1466 2.3223>, rot=<0.7289 0.0465 -0.6829 0.009>}
 upperarm_r {parent=clavicle_r, loc=<-15.5111 0 0>, rot=<0.0305 -0.536 -0.0057 -0.8437>}
 lowerarm_r {parent=upperarm_r, loc=<-26.6515 0 0>, rot=<0.0391 -0.0099 -0.2453 -0.9686>}
 hand_r {parent=lowerarm_r, loc=<-22.3492 0 0>, rot=<-0.7081 -0.0507 0.0975 -0.6975>}
 index_metacarpal_r {parent=hand_r, loc=<-1.6534 -0.0181 2.9601>, rot=<0.0158 -0.0124 -0.0173 -0.9996>}
 index_01_r {parent=index_metacarpal_r, loc=<-8.3008 -0.0357 0.0524>, rot=<-0.0044 0.0485 0.0924 -0.9945>}
 index_02_r {parent=index_01_r, loc=<-3.3589 0.3003 -0.0453>, rot=<-0.002 0.0027 0.3225 -0.9466>}
 index_03_r {parent=index_02_r, loc=<-2.5578 0.1231 0.0251>, rot=<-0.002 0.0067 0.2558 -0.9667>}
 middle_metacarpal_r {parent=hand_r, loc=<-2.1856 0.1726 0.5984>, rot=<0.006 0.0168 -0.0329 -0.9993>}
 middle_01_r {parent=middle_metacarpal_r, loc=<-7.3273 0.2447 -0.0129>, rot=<-0.0208 0.0042 0.0858 -0.9961>}
 middle_02_r {parent=middle_01_r, loc=<-4.0142 -0.1411 -0.07>, rot=<-0.0217 0.0259 0.3512 -0.9357>}
 middle_03_r {parent=middle_02_r, loc=<-2.6074 -0.0534 -0.1312>, rot=<0.0083 0.0188 0.1708 -0.9851>}
 pinky_metacarpal_r {parent=hand_r, loc=<-1.7617 0.9753 -3.3814>, rot=<-0.0397 0.0814 -0.069 -0.9935>}
 pinky_01_r {parent=pinky_metacarpal_r, loc=<-7.1707 0.4832 0.1094>, rot=<-0.0654 -0.0035 0.2242 -0.9723>}
 pinky_02_r {parent=pinky_01_r, loc=<-2.8274 -0.0892 0.1043>, rot=<0.0037 0.0011 0.2854 -0.9584>}
 pinky_03_r {parent=pinky_02_r, loc=<-2.2045 -0.031 0.0078>, rot=<-0.0002 0.0098 0.1445 -0.9895>}
 ring_metacarpal_r {parent=hand_r, loc=<-2.2073 0.3627 -1.3638>, rot=<-0.0284 0.0722 -0.0458 -0.9959>}
 ring_01_r {parent=ring_metacarpal_r, loc=<-7.0297 0.6197 0.1446>, rot=<-0.0489 -0.0252 0.1892 -0.9804>}
 ring_02_r {parent=ring_01_r, loc=<-3.6171 -0.4648 0.1027>, rot=<0 0 0.2806 -0.9598>}
 ring_03_r {parent=ring_02_r, loc=<-2.4153 -0.1141 -0.0199>, rot=<0 0 0.1189 -0.9929>}
 thumb_01_r {parent=hand_r, loc=<-1.3119 0.3193 2.459>, rot=<0.6546 -0.4042 0.0664 -0.6354>}
 thumb_02_r {parent=thumb_01_r, loc=<-5.9335 0.0447 -0.0401>, rot=<-0.0037 0.0213 0.298 -0.9543>}
 thumb_03_r {parent=thumb_02_r, loc=<-2.2223 -0.3682 0.0907>, rot=<0.0118 -0.013 0.0066 -0.9998>}
 weapon_r {parent=hand_r, loc=<0.1975 3.4112 -0.3811>, rot=<0 0 0 -1>}
 lowerarm_twist_01_r {parent=lowerarm_r, loc=<-17.855 0 0>, rot=<0 0 0 1>}
 lowerarm_twist_02_r {parent=lowerarm_r, loc=<-9.6199 0 0>, rot=<0 0 0 1>}
 upperarm_twist_01_r {parent=upperarm_r, loc=<-0.0041 0 0>, rot=<0 0 0 1>}
 upperarm_twist_02_r {parent=upperarm_r, loc=<-13.7942 0 0>, rot=<0 0 0 1>}
 dyn_shoulderpad_r {parent=clavicle_r, loc=<-7.1878 2.4274 -11.5862>, rot=<0.9991 0.0252 0.0336 0.0008>}
 neck_01 {parent=spine_05, loc=<11.3138 0.1529 0>, rot=<0 0 -0.2003 -0.9797>}
 neck_02 {parent=neck_01, loc=<4.5627 0 0>, rot=<0 0 0.0504 -0.9987>}
 head {parent=neck_02, loc=<5.3914 0 0>, rot=<0 0 0.115 -0.9934>}
 dyn_earing_l {parent=head, loc=<13.3224 -1.2499 -11.0254>, rot=<0.0029 -0.7071 -0.0029 -0.7071>}
 dyn_earing_r {parent=head, loc=<13.3224 -1.2499 11.0254>, rot=<0.0029 -0.7071 -0.0029 -0.7071>}
 dyn_hair_01 {parent=head, loc=<14.5128 15.9305 1.4919>, rot=<0.0029 -0.7071 -0.0029 -0.7071>}
 dyn_hair_02 {parent=head, loc=<19.4179 10.9915 0.4339>, rot=<0.0029 -0.7071 -0.0029 -0.7071>}
 dyn_hair_03 {parent=head, loc=<21.5942 2.0225 1.2273>, rot=<0.0029 -0.7071 -0.0029 -0.7071>}
 dyn_hair_04 {parent=head, loc=<20.2035 -3.5032 2.4886>, rot=<0.0029 -0.7071 -0.0029 -0.7071>}
 dyn_hair_05 {parent=head, loc=<15.846 -7.1285 1.0201>, rot=<0.0029 -0.7071 -0.0029 -0.7071>}
 dyn_hair_06 {parent=head, loc=<13.0411 -7.5297 0.6464>, rot=<0.0029 -0.7071 -0.0029 -0.7071>}
 dyn_hair_07 {parent=head, loc=<14.4839 -10.6874 -0.2481>, rot=<0.0029 -0.7071 -0.0029 -0.7071>}
 dyn_hair_08 {parent=head, loc=<10.8164 -15.5045 -0.9599>, rot=<0.0029 -0.7071 -0.0029 -0.7071>}
 dyn_hair_09 {parent=head, loc=<2.8279 -20.2546 -0.0271>, rot=<0.0029 -0.7071 -0.0029 -0.7071>}
 dyn_mane_04 {parent=neck_01, loc=<-0.7209 -20.4935 0.6717>, rot=<-0.1137 -0.6979 0.1137 -0.6979>}
 dyn_outer_collar_r {parent=spine_05, loc=<6.6797 13.3489 11.5755>, rot=<0.0284 -0.7065 -0.0284 -0.7065>}
 dyn_outer_collar_l {parent=spine_05, loc=<6.6797 13.3489 -11.5755>, rot=<0.0284 -0.7065 -0.0284 -0.7065>}
 dyn_mane_05 {parent=spine_05, loc=<9.2776 -18.5539 -0.022>, rot=<0.0284 -0.7065 -0.0284 -0.7065>}
 dyn_strap {parent=spine_05, loc=<-1.4861 -14.4845 13.4839>, rot=<0.0284 -0.7065 -0.0284 -0.7065>}
 dyn_coat_bk_01 {parent=spine_03, loc=<6.153 -14.4903 -4.2158>, rot=<0.0104 -0.707 -0.0104 -0.707>}
 dyn_coat_out_01_l {parent=spine_03, loc=<1.4871 -1.6511 -14.594>, rot=<0.0104 -0.707 -0.0104 -0.707>}
 dyn_coat_fr_01_l {parent=spine_03, loc=<3.519 13.0393 -9.2365>, rot=<0.0104 -0.707 -0.0104 -0.707>}
 dyn_coat_bk_01_l {parent=spine_03, loc=<4.3001 -9.2187 -13.7833>, rot=<0.0104 -0.707 -0.0104 -0.707>}
 dyn_coat_frdiag_01_r {parent=spine_03, loc=<2.3217 4.6222 15.473>, rot=<0.0104 -0.707 -0.0104 -0.707>}
 dyn_coat_fr_01_r {parent=spine_03, loc=<0.8199 11.6687 9.471>, rot=<0.0104 -0.707 -0.0104 -0.707>}
 dyn_coat_bk_01_r {parent=spine_03, loc=<4.3001 -9.2187 13.7833>, rot=<0.0104 -0.707 -0.0104 -0.707>}
 thigh_l {parent=pelvis, loc=<-2.5693 -0.0109 -7.7501>, rot=<0.0012 0.0572 0.0155 -0.9982>}
 calf_l {parent=thigh_l, loc=<-39.0484 0 0>, rot=<0.0006 -0.0161 -0.0391 -0.9991>}
 calf_twist_01_l {parent=calf_l, loc=<-35.8131 -0.0967 0.0397>, rot=<0.0237 0.0095 -0.0035 -0.9997>}
 calf_twist_02_l {parent=calf_l, loc=<-18.8064 -0.0967 0.0397>, rot=<0.0237 0.0095 -0.0035 -0.9997>}
 foot_l {parent=calf_l, loc=<-42.561 0.1077 0.4292>, rot=<0.0418 -0.0455 0.0343 -0.9975>}
 ball_l {parent=foot_l, loc=<-6.1599 13.8264 0.6806>, rot=<0.0001 -0.0001 -0.7063 -0.7079>}
 thigh_twist_01_l {parent=thigh_l, loc=<-21.6303 0 0>, rot=<0.0013 0 0 -1>}
 thigh_r {parent=pelvis, loc=<-2.5691 -0.0109 7.7501>, rot=<0.0572 -0.0012 0.9982 0.0155>}
 calf_r {parent=thigh_r, loc=<39.0487 0 0>, rot=<0.0006 -0.0161 -0.0391 -0.9991>}
 calf_twist_01_r {parent=calf_r, loc=<35.8132 0.0967 -0.0397>, rot=<0.0237 0.0095 -0.0035 -0.9997>}
 calf_twist_02_r {parent=calf_r, loc=<18.8064 0.0967 -0.0397>, rot=<0.0237 0.0095 -0.0035 -0.9997>}
 foot_r {parent=calf_r, loc=<42.5611 -0.1077 -0.4292>, rot=<0.0418 -0.0455 0.0343 -0.9975>}
 ball_r {parent=foot_r, loc=<6.1598 -13.8264 -0.6806>, rot=<0.0001 -0.0001 -0.7063 -0.7079>}
 thigh_twist_01_r {parent=thigh_r, loc=<21.6303 0 0>, rot=<0.0013 0 0 -1>}
 wolf_thigh_l {parent=pelvis, loc=<-2.5693 -0.0109 -7.7501>, rot=<0.0012 0.0572 0.0155 -0.9982>}
 wolf_calf_l {parent=wolf_thigh_l, loc=<-36.3283 2.1181 0.0962>, rot=<0.0006 -0.0161 -0.0391 -0.9991>}
 wolf_calf_twist_01_l {parent=wolf_calf_l, loc=<-32.6056 -5.0853 0.7725>, rot=<0.0237 0.0095 -0.0035 -0.9997>}
 wolf_calf_twist_02_l {parent=wolf_calf_l, loc=<-10.4852 -4.0769 0.3061>, rot=<0.0237 0.0095 -0.0035 -0.9997>}
 wolf_foot_l {parent=wolf_calf_l, loc=<-23.7181 -7.4961 0.2405>, rot=<0.0418 -0.0455 0.0343 -0.9975>}
 wolf_ball_l {parent=wolf_foot_l, loc=<-19.3349 3.4907 -0.8786>, rot=<0.0001 -0.0001 -0.7063 -0.7079>}
 wolf_toe_l {parent=wolf_ball_l, loc=<8.8736 8.7423 -1.1808>, rot=<0 0 0 1>}
 toe_claw_l {parent=wolf_ball_l, loc=<0.001 2.6176 3.0483>, rot=<-0.5229 -0.4735 0.4801 -0.5214>}
 wolf_thigh_twist_01_l {parent=wolf_thigh_l, loc=<-21.6303 0 0>, rot=<0.0013 0 0 -1>}
 wolf_thigh_r {parent=pelvis, loc=<-2.5694 -0.0109 7.7501>, rot=<0.0572 -0.0012 0.9982 0.0155>}
 wolf_calf_r {parent=wolf_thigh_r, loc=<36.3283 -2.1181 -0.0962>, rot=<0.0006 -0.0161 -0.0391 -0.9991>}
 wolf_calf_twist_01_r {parent=wolf_calf_r, loc=<32.6056 5.0853 -0.7724>, rot=<0.0237 0.0095 -0.0035 -0.9997>}
 wolf_calf_twist_02_r {parent=wolf_calf_r, loc=<10.4852 4.0769 -0.3061>, rot=<0.0237 0.0095 -0.0035 -0.9997>}
 wolf_foot_r {parent=wolf_calf_r, loc=<23.7181 7.4961 -0.2404>, rot=<0.0418 -0.0455 0.0343 -0.9975>}
 wolf_ball_r {parent=wolf_foot_r, loc=<19.3348 -3.4907 0.8786>, rot=<0.0001 -0.0001 -0.7063 -0.7079>}
 wolf_toe_r {parent=wolf_ball_r, loc=<-8.8736 -8.7424 1.1809>, rot=<0 0 0 -1>}
 toe_claw_r {parent=wolf_ball_r, loc=<-0.001 -2.6176 -3.0482>, rot=<-0.5229 -0.4735 0.4801 -0.5214>}
 wolf_thigh_twist_01_r {parent=wolf_thigh_r, loc=<21.6303 0 0>, rot=<0.0013 0 0 -1>}
 const_nunchuck_base {parent=pelvis, loc=<4.5103 -6.2979 -12.8162>, rot=<-0.1616 -0.6196 -0.1798 -0.7467>}
 dyn_upper_nunchuck_01 {parent=const_nunchuck_base, loc=<0.799 1.8312 2.1391>, rot=<-0.0068 -0.0519 0.1303 -0.9901>}
 dyn_upper_nunchuck_02 {parent=dyn_upper_nunchuck_01, loc=<1.8445 -0.0117 0.0228>, rot=<0.0083 -0.4771 0.0153 -0.8787>}
 dyn_upper_nunchuck_03 {parent=dyn_upper_nunchuck_02, loc=<2.9661 0.0091 0.4937>, rot=<-0.0042 0.2419 -0.0169 -0.9701>}
 dyn_lower_nunchuck_01 {parent=const_nunchuck_base, loc=<-0.6937 -0.3678 -2.1511>, rot=<-0.387 -0.3987 0.642 -0.5282>}
 dyn_lower_nunchuck_02 {parent=dyn_lower_nunchuck_01, loc=<1.8412 -0.0225 -0.2072>, rot=<-0.0004 0.0175 -0.007 -0.9998>}
 dyn_lower_nunchuck_03 {parent=dyn_lower_nunchuck_02, loc=<2.9583 0.0133 -0.0048>, rot=<0.0007 0 0.0087 -1>}
 dyn_lower_nunchuck_04 {parent=dyn_lower_nunchuck_03, loc=<2.7036 -0.0301 0.0332>, rot=<-0.0009 -0.0393 -0.0218 -0.999>}
 ik_foot_root {parent=root, loc=<0 0 0>, rot=<0 0 0 1>}
 ik_foot_l {parent=ik_foot_root, loc=<16.4438 3.5592 3.2853>, rot=<-0.5 -0.5 0.5 0.5>}
 ik_foot_r {parent=ik_foot_root, loc=<-16.4438 3.5592 3.2853>, rot=<-0.5 0.5 0.5 -0.5>}
 ik_hand_root {parent=root, loc=<0 0 0>, rot=<0 0 0 1>}
 ik_hand_gun {parent=ik_hand_root, loc=<-33.8029 3.8283 90.5694>, rot=<-0.4834 -0.3223 -0.5887 0.5621>}
 ik_hand_l {parent=ik_hand_gun, loc=<6.7067 65.8041 13.978>, rot=<-0.0992 -0.9733 -0.2068 0>}
 ik_hand_r {parent=ik_hand_gun, loc=<0 0 0>, rot=<0 0 0 -1>}
]
variables [
 BeginExecution_16_0 {type=FRigUnit_BeginExecution, default=()}
 BeginExecution_16_0_6 {type=FRigUnit_BeginExecution, default=()}
 FABRIK_0_0 {
  type=FRigUnit_FABRIK
  default=(StartBone="",EffectorBone="",EffectorTransform=(Rotation=(X=0,Y=0,Z=0,W=1),Translation=(X=0,Y=0,Z=0),Scale3D=(X=1,Y=1,Z=1)),Precision=1,Weight=1,bPropagateToChildren=False,MaxIterations=10,bSetEffectorTransform=True)
 }
 FABRIK_1 {
  type=FRigUnit_FABRIK
  default=(StartBone="",EffectorBone="",EffectorTransform=(Rotation=(X=0,Y=0,Z=0,W=1),Translation=(X=0,Y=0,Z=0),Scale3D=(X=1,Y=1,Z=1)),Precision=1,Weight=1,bPropagateToChildren=False,MaxIterations=10,bSetEffectorTransform=True)
 }
 GetBoneTransform {
  type=FRigUnit_GetBoneTransform
  default=(Bone="",Space=GlobalSpace,Transform=(Rotation=(X=0,Y=0,Z=0,W=1),Translation=(X=0,Y=0,Z=0),Scale3D=(X=1,Y=1,Z=1)))
 }
 GetBoneTransform_0_0_0 {
  type=FRigUnit_GetBoneTransform
  default=(Bone="",Space=GlobalSpace,Transform=(Rotation=(X=0,Y=0,Z=0,W=1),Translation=(X=0,Y=0,Z=0),Scale3D=(X=1,Y=1,Z=1)))
 }
 GetBoneTransform_0_1 {
  type=FRigUnit_GetBoneTransform
  default=(Bone="",Space=GlobalSpace,Transform=(Rotation=(X=0,Y=0,Z=0,W=1),Translation=(X=0,Y=0,Z=0),Scale3D=(X=1,Y=1,Z=1)))
 }
 GetBoneTransform_1_0_0 {
  type=FRigUnit_GetBoneTransform
  default=(Bone="",Space=GlobalSpace,Transform=(Rotation=(X=0,Y=0,Z=0,W=1),Translation=(X=0,Y=0,Z=0),Scale3D=(X=1,Y=1,Z=1)))
 }
 GetBoneTransform_1_1 {
  type=FRigUnit_GetBoneTransform
  default=(Bone="",Space=GlobalSpace,Transform=(Rotation=(X=0,Y=0,Z=0,W=1),Translation=(X=0,Y=0,Z=0),Scale3D=(X=1,Y=1,Z=1)))
 }
 GetBoneTransform_2 {
  type=FRigUnit_GetBoneTransform
  default=(Bone="",Space=GlobalSpace,Transform=(Rotation=(X=0,Y=0,Z=0,W=1),Translation=(X=0,Y=0,Z=0),Scale3D=(X=1,Y=1,Z=1)))
 }
 GetBoneTransform_3_0_0 {
  type=FRigUnit_GetBoneTransform
  default=(Bone="",Space=GlobalSpace,Transform=(Rotation=(X=0,Y=0,Z=0,W=1),Translation=(X=0,Y=0,Z=0),Scale3D=(X=1,Y=1,Z=1)))
 }
 GetBoneTransform_3_1 {
  type=FRigUnit_GetBoneTransform
  default=(Bone="",Space=GlobalSpace,Transform=(Rotation=(X=0,Y=0,Z=0,W=1),Translation=(X=0,Y=0,Z=0),Scale3D=(X=1,Y=1,Z=1)))
 }
 GetInitialBoneTransform {
  type=FRigUnit_GetInitialBoneTransform
  default=(Bone="",Space=GlobalSpace,Transform=(Rotation=(X=0,Y=0,Z=0,W=1),Translation=(X=0,Y=0,Z=0),Scale3D=(X=1,Y=1,Z=1)))
 }
 GetInitialBoneTransform_0 {
  type=FRigUnit_GetInitialBoneTransform
  default=(Bone="",Space=GlobalSpace,Transform=(Rotation=(X=0,Y=0,Z=0,W=1),Translation=(X=0,Y=0,Z=0),Scale3D=(X=1,Y=1,Z=1)))
 }
 GetInitialBoneTransform_0_0 {
  type=FRigUnit_GetInitialBoneTransform
  default=(Bone="",Space=GlobalSpace,Transform=(Rotation=(X=0,Y=0,Z=0,W=1),Translation=(X=0,Y=0,Z=0),Scale3D=(X=1,Y=1,Z=1)))
 }
 GetInitialBoneTransform_1_0_0_0 {
  type=FRigUnit_GetInitialBoneTransform
  default=(Bone="",Space=GlobalSpace,Transform=(Rotation=(X=0,Y=0,Z=0,W=1),Translation=(X=0,Y=0,Z=0),Scale3D=(X=1,Y=1,Z=1)))
 }
 GetInitialBoneTransform_1_0_1_0_0 {
  type=FRigUnit_GetInitialBoneTransform
  default=(Bone="",Space=GlobalSpace,Transform=(Rotation=(X=0,Y=0,Z=0,W=1),Translation=(X=0,Y=0,Z=0),Scale3D=(X=1,Y=1,Z=1)))
 }
 GetInitialBoneTransform_1_0_1_1 {
  type=FRigUnit_GetInitialBoneTransform
  default=(Bone="",Space=GlobalSpace,Transform=(Rotation=(X=0,Y=0,Z=0,W=1),Translation=(X=0,Y=0,Z=0),Scale3D=(X=1,Y=1,Z=1)))
 }
 GetInitialBoneTransform_1_0_2 {
  type=FRigUnit_GetInitialBoneTransform
  default=(Bone="",Space=GlobalSpace,Transform=(Rotation=(X=0,Y=0,Z=0,W=1),Translation=(X=0,Y=0,Z=0),Scale3D=(X=1,Y=1,Z=1)))
 }
 GetInitialBoneTransform_1_1_0 {
  type=FRigUnit_GetInitialBoneTransform
  default=(Bone="",Space=GlobalSpace,Transform=(Rotation=(X=0,Y=0,Z=0,W=1),Translation=(X=0,Y=0,Z=0),Scale3D=(X=1,Y=1,Z=1)))
 }
 GetInitialBoneTransform_1_2 {
  type=FRigUnit_GetInitialBoneTransform
  default=(Bone="",Space=GlobalSpace,Transform=(Rotation=(X=0,Y=0,Z=0,W=1),Translation=(X=0,Y=0,Z=0),Scale3D=(X=1,Y=1,Z=1)))
 }
 GetInitialBoneTransform_2 {
  type=FRigUnit_GetInitialBoneTransform
  default=(Bone="",Space=GlobalSpace,Transform=(Rotation=(X=0,Y=0,Z=0,W=1),Translation=(X=0,Y=0,Z=0),Scale3D=(X=1,Y=1,Z=1)))
 }
 GetInitialBoneTransform_2_0_0_0 {
  type=FRigUnit_GetInitialBoneTransform
  default=(Bone="",Space=GlobalSpace,Transform=(Rotation=(X=0,Y=0,Z=0,W=1),Translation=(X=0,Y=0,Z=0),Scale3D=(X=1,Y=1,Z=1)))
 }
 GetInitialBoneTransform_2_0_1_0_0 {
  type=FRigUnit_GetInitialBoneTransform
  default=(Bone="",Space=GlobalSpace,Transform=(Rotation=(X=0,Y=0,Z=0,W=1),Translation=(X=0,Y=0,Z=0),Scale3D=(X=1,Y=1,Z=1)))
 }
 GetInitialBoneTransform_2_0_1_1 {
  type=FRigUnit_GetInitialBoneTransform
  default=(Bone="",Space=GlobalSpace,Transform=(Rotation=(X=0,Y=0,Z=0,W=1),Translation=(X=0,Y=0,Z=0),Scale3D=(X=1,Y=1,Z=1)))
 }
 GetInitialBoneTransform_2_0_2 {
  type=FRigUnit_GetInitialBoneTransform
  default=(Bone="",Space=GlobalSpace,Transform=(Rotation=(X=0,Y=0,Z=0,W=1),Translation=(X=0,Y=0,Z=0),Scale3D=(X=1,Y=1,Z=1)))
 }
 GetInitialBoneTransform_2_1 {
  type=FRigUnit_GetInitialBoneTransform
  default=(Bone="",Space=GlobalSpace,Transform=(Rotation=(X=0,Y=0,Z=0,W=1),Translation=(X=0,Y=0,Z=0),Scale3D=(X=1,Y=1,Z=1)))
 }
 GetInitialBoneTransform_3 {
  type=FRigUnit_GetInitialBoneTransform
  default=(Bone="",Space=GlobalSpace,Transform=(Rotation=(X=0,Y=0,Z=0,W=1),Translation=(X=0,Y=0,Z=0),Scale3D=(X=1,Y=1,Z=1)))
 }
 MathTransformMakeRelative {
  type=FRigVMFunction_MathTransformMakeRelative
  default=(Global=(Rotation=(X=0,Y=0,Z=0,W=1),Translation=(X=0,Y=0,Z=0),Scale3D=(X=1,Y=1,Z=1)),Parent=(Rotation=(X=0,Y=0,Z=0,W=1),Translation=(X=0,Y=0,Z=0),Scale3D=(X=1,Y=1,Z=1)),Local=(Rotation=(X=0,Y=0,Z=0,W=1),Translation=(X=0,Y=0,Z=0),Scale3D=(X=1,Y=1,Z=1)))
 }
 MathTransformMakeRelative_0_0_0_0 {
  type=FRigVMFunction_MathTransformMakeRelative
  default=(Global=(Rotation=(X=0,Y=0,Z=0,W=1),Translation=(X=0,Y=0,Z=0),Scale3D=(X=1,Y=1,Z=1)),Parent=(Rotation=(X=0,Y=0,Z=0,W=1),Translation=(X=0,Y=0,Z=0),Scale3D=(X=1,Y=1,Z=1)),Local=(Rotation=(X=0,Y=0,Z=0,W=1),Translation=(X=0,Y=0,Z=0),Scale3D=(X=1,Y=1,Z=1)))
 }
 MathTransformMakeRelative_0_0_1_0_0 {
  type=FRigVMFunction_MathTransformMakeRelative
  default=(Global=(Rotation=(X=0,Y=0,Z=0,W=1),Translation=(X=0,Y=0,Z=0),Scale3D=(X=1,Y=1,Z=1)),Parent=(Rotation=(X=0,Y=0,Z=0,W=1),Translation=(X=0,Y=0,Z=0),Scale3D=(X=1,Y=1,Z=1)),Local=(Rotation=(X=0,Y=0,Z=0,W=1),Translation=(X=0,Y=0,Z=0),Scale3D=(X=1,Y=1,Z=1)))
 }
 MathTransformMakeRelative_0_0_1_1 {
  type=FRigVMFunction_MathTransformMakeRelative
  default=(Global=(Rotation=(X=0,Y=0,Z=0,W=1),Translation=(X=0,Y=0,Z=0),Scale3D=(X=1,Y=1,Z=1)),Parent=(Rotation=(X=0,Y=0,Z=0,W=1),Translation=(X=0,Y=0,Z=0),Scale3D=(X=1,Y=1,Z=1)),Local=(Rotation=(X=0,Y=0,Z=0,W=1),Translation=(X=0,Y=0,Z=0),Scale3D=(X=1,Y=1,Z=1)))
 }
 MathTransformMakeRelative_0_0_2 {
  type=FRigVMFunction_MathTransformMakeRelative
  default=(Global=(Rotation=(X=0,Y=0,Z=0,W=1),Translation=(X=0,Y=0,Z=0),Scale3D=(X=1,Y=1,Z=1)),Parent=(Rotation=(X=0,Y=0,Z=0,W=1),Translation=(X=0,Y=0,Z=0),Scale3D=(X=1,Y=1,Z=1)),Local=(Rotation=(X=0,Y=0,Z=0,W=1),Translation=(X=0,Y=0,Z=0),Scale3D=(X=1,Y=1,Z=1)))
 }
 MathTransformMakeRelative_0_1_0 {
  type=FRigVMFunction_MathTransformMakeRelative
  default=(Global=(Rotation=(X=0,Y=0,Z=0,W=1),Translation=(X=0,Y=0,Z=0),Scale3D=(X=1,Y=1,Z=1)),Parent=(Rotation=(X=0,Y=0,Z=0,W=1),Translation=(X=0,Y=0,Z=0),Scale3D=(X=1,Y=1,Z=1)),Local=(Rotation=(X=0,Y=0,Z=0,W=1),Translation=(X=0,Y=0,Z=0),Scale3D=(X=1,Y=1,Z=1)))
 }
 MathTransformMakeRelative_0_2 {
  type=FRigVMFunction_MathTransformMakeRelative
  default=(Global=(Rotation=(X=0,Y=0,Z=0,W=1),Translation=(X=0,Y=0,Z=0),Scale3D=(X=1,Y=1,Z=1)),Parent=(Rotation=(X=0,Y=0,Z=0,W=1),Translation=(X=0,Y=0,Z=0),Scale3D=(X=1,Y=1,Z=1)),Local=(Rotation=(X=0,Y=0,Z=0,W=1),Translation=(X=0,Y=0,Z=0),Scale3D=(X=1,Y=1,Z=1)))
 }
 MathTransformMakeRelative_1 {
  type=FRigVMFunction_MathTransformMakeRelative
  default=(Global=(Rotation=(X=0,Y=0,Z=0,W=1),Translation=(X=0,Y=0,Z=0),Scale3D=(X=1,Y=1,Z=1)),Parent=(Rotation=(X=0,Y=0,Z=0,W=1),Translation=(X=0,Y=0,Z=0),Scale3D=(X=1,Y=1,Z=1)),Local=(Rotation=(X=0,Y=0,Z=0,W=1),Translation=(X=0,Y=0,Z=0),Scale3D=(X=1,Y=1,Z=1)))
 }
 MathTransformMakeRelative_1_0_0_0 {
  type=FRigVMFunction_MathTransformMakeRelative
  default=(Global=(Rotation=(X=0,Y=0,Z=0,W=1),Translation=(X=0,Y=0,Z=0),Scale3D=(X=1,Y=1,Z=1)),Parent=(Rotation=(X=0,Y=0,Z=0,W=1),Translation=(X=0,Y=0,Z=0),Scale3D=(X=1,Y=1,Z=1)),Local=(Rotation=(X=0,Y=0,Z=0,W=1),Translation=(X=0,Y=0,Z=0),Scale3D=(X=1,Y=1,Z=1)))
 }
 MathTransformMakeRelative_1_0_1 {
  type=FRigVMFunction_MathTransformMakeRelative
  default=(Global=(Rotation=(X=0,Y=0,Z=0,W=1),Translation=(X=0,Y=0,Z=0),Scale3D=(X=1,Y=1,Z=1)),Parent=(Rotation=(X=0,Y=0,Z=0,W=1),Translation=(X=0,Y=0,Z=0),Scale3D=(X=1,Y=1,Z=1)),Local=(Rotation=(X=0,Y=0,Z=0,W=1),Translation=(X=0,Y=0,Z=0),Scale3D=(X=1,Y=1,Z=1)))
 }
 MathTransformMakeRelative_1_1_0_0 {
  type=FRigVMFunction_MathTransformMakeRelative
  default=(Global=(Rotation=(X=0,Y=0,Z=0,W=1),Translation=(X=0,Y=0,Z=0),Scale3D=(X=1,Y=1,Z=1)),Parent=(Rotation=(X=0,Y=0,Z=0,W=1),Translation=(X=0,Y=0,Z=0),Scale3D=(X=1,Y=1,Z=1)),Local=(Rotation=(X=0,Y=0,Z=0,W=1),Translation=(X=0,Y=0,Z=0),Scale3D=(X=1,Y=1,Z=1)))
 }
 MathTransformMakeRelative_1_1_1 {
  type=FRigVMFunction_MathTransformMakeRelative
  default=(Global=(Rotation=(X=0,Y=0,Z=0,W=1),Translation=(X=0,Y=0,Z=0),Scale3D=(X=1,Y=1,Z=1)),Parent=(Rotation=(X=0,Y=0,Z=0,W=1),Translation=(X=0,Y=0,Z=0),Scale3D=(X=1,Y=1,Z=1)),Local=(Rotation=(X=0,Y=0,Z=0,W=1),Translation=(X=0,Y=0,Z=0),Scale3D=(X=1,Y=1,Z=1)))
 }
 MathTransformMakeRelative_1_2 {
  type=FRigVMFunction_MathTransformMakeRelative
  default=(Global=(Rotation=(X=0,Y=0,Z=0,W=1),Translation=(X=0,Y=0,Z=0),Scale3D=(X=1,Y=1,Z=1)),Parent=(Rotation=(X=0,Y=0,Z=0,W=1),Translation=(X=0,Y=0,Z=0),Scale3D=(X=1,Y=1,Z=1)),Local=(Rotation=(X=0,Y=0,Z=0,W=1),Translation=(X=0,Y=0,Z=0),Scale3D=(X=1,Y=1,Z=1)))
 }
 MathTransformMakeRelative_2 {
  type=FRigVMFunction_MathTransformMakeRelative
  default=(Global=(Rotation=(X=0,Y=0,Z=0,W=1),Translation=(X=0,Y=0,Z=0),Scale3D=(X=1,Y=1,Z=1)),Parent=(Rotation=(X=0,Y=0,Z=0,W=1),Translation=(X=0,Y=0,Z=0),Scale3D=(X=1,Y=1,Z=1)),Local=(Rotation=(X=0,Y=0,Z=0,W=1),Translation=(X=0,Y=0,Z=0),Scale3D=(X=1,Y=1,Z=1)))
 }
 MathTransformMul_0_0 {
  type=FRigVMFunction_MathTransformMul
  default=(A=(Rotation=(X=0,Y=0,Z=0,W=1),Translation=(X=0,Y=0,Z=0),Scale3D=(X=1,Y=1,Z=1)),B=(Rotation=(X=0,Y=0,Z=0,W=1),Translation=(X=0,Y=0,Z=0),Scale3D=(X=1,Y=1,Z=1)),Result=(Rotation=(X=0,Y=0,Z=0,W=1),Translation=(X=0,Y=0,Z=0),Scale3D=(X=1,Y=1,Z=1)))
 }
 MathTransformMul_1 {
  type=FRigVMFunction_MathTransformMul
  default=(A=(Rotation=(X=0,Y=0,Z=0,W=1),Translation=(X=0,Y=0,Z=0),Scale3D=(X=1,Y=1,Z=1)),B=(Rotation=(X=0,Y=0,Z=0,W=1),Translation=(X=0,Y=0,Z=0),Scale3D=(X=1,Y=1,Z=1)),Result=(Rotation=(X=0,Y=0,Z=0,W=1),Translation=(X=0,Y=0,Z=0),Scale3D=(X=1,Y=1,Z=1)))
 }
 SetBoneRotation {type=FRigUnit_SetBoneRotation, default=(Bone="",Rotation=(X=0,Y=0,Z=0,W=1),Space=LocalSpace,Weight=1,bPropagateToChildren=False)}
 SetBoneRotation_0_0_0_0 {type=FRigUnit_SetBoneRotation, default=(Bone="",Rotation=(X=0,Y=0,Z=0,W=1),Space=LocalSpace,Weight=1,bPropagateToChildren=False)}
 SetBoneRotation_0_0_1 {type=FRigUnit_SetBoneRotation, default=(Bone="",Rotation=(X=0,Y=0,Z=0,W=1),Space=LocalSpace,Weight=1,bPropagateToChildren=False)}
 SetBoneRotation_0_1_0 {type=FRigUnit_SetBoneRotation, default=(Bone="",Rotation=(X=0,Y=0,Z=0,W=1),Space=LocalSpace,Weight=1,bPropagateToChildren=False)}
 SetBoneRotation_0_2 {type=FRigUnit_SetBoneRotation, default=(Bone="",Rotation=(X=0,Y=0,Z=0,W=1),Space=LocalSpace,Weight=1,bPropagateToChildren=False)}
 SetBoneRotation_1 {type=FRigUnit_SetBoneRotation, default=(Bone="",Rotation=(X=0,Y=0,Z=0,W=1),Space=LocalSpace,Weight=1,bPropagateToChildren=False)}
 SetBoneTransform_0_0 {
  type=FRigUnit_SetBoneTransform
  default=(Bone="",Transform=(Rotation=(X=0,Y=0,Z=0,W=1),Translation=(X=0,Y=0,Z=0),Scale3D=(X=1,Y=1,Z=1)),Result=(Rotation=(X=0,Y=0,Z=0,W=1),Translation=(X=0,Y=0,Z=0),Scale3D=(X=1,Y=1,Z=1)),Space=LocalSpace,Weight=1,bPropagateToChildren=False)
 }
 SetBoneTransform_1 {
  type=FRigUnit_SetBoneTransform
  default=(Bone="",Transform=(Rotation=(X=0,Y=0,Z=0,W=1),Translation=(X=0,Y=0,Z=0),Scale3D=(X=1,Y=1,Z=1)),Result=(Rotation=(X=0,Y=0,Z=0,W=1),Translation=(X=0,Y=0,Z=0),Scale3D=(X=1,Y=1,Z=1)),Space=LocalSpace,Weight=1,bPropagateToChildren=False)
 }
]
