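# Secondary-motion anim blueprint for the male Werewolf 02 player body, layered on the pose copied from the attached parent mesh.
# A Control Rig feeds AnimDynamics for coat, collar and shoulder pads, then mane constraints, into a cached pose.
# Two RigidBody passes blend on IsCrouching (crouched and stationary), switching to the secondary physics asset, before a neck_02 offset and scale.
# Two other AnimDynamics chains are unconnected to the output.
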AnimBlueprint /Game/Characters/Player/Male/Medium/Bodies/M_MED_Werewolf_01/Meshes/M_MED_Werewolf_02_Body_AnimBP
target_skeleton=&/M_MED_Werewolf_02_Body_Skeleton
graphs [
 AnimGraph {nodes [
   n1 = Anim Dynamics - Bone: dyn_coat_bk_01(Alpha=0.95, ExternalForce="0,0,0") {
    angularDampingOverride=1
    angularSpringConstant=2000
    bAngularSpring=true
    bGravityOverrideInSimSpace=true
    gravityScale=0.5
    lODThreshold=1
    linearDampingOverride=1
    linearSpringConstant=200
    alphaBoolBlend {blendOption=Linear}
    boundBone {boneName=dyn_coat_bk_01}
    componentLinearAccScale=<1 1 1>
    componentLinearVelScale=<1 1 1>
    physicsBodyDefinitions [{
      sphereCollisionRadius=0
      boundBone {boneName=dyn_coat_bk_01}
      boxExtents=<20 20 10>
      constraintSetup {angularTargetAxis=AxisY, twistAxis=AxisZ, angularLimitsMax=<30 30 0>, angularLimitsMin=<-30 -30 0>, angularTarget=<0 0 0>}
      localJointOffset=<0 0 -1>
     }]
   }
   n2 = Anim Dynamics - Bone: dyn_coat_out_01_l(ComponentPose=@n1, Alpha=1, ExternalForce="0,0,0") {
    angularDampingOverride=1
    angularSpringConstant=2000
    bAngularSpring=true
    bGravityOverrideInSimSpace=true
    gravityScale=0.5
    lODThreshold=1
    linearDampingOverride=1
    linearSpringConstant=200
    alphaBoolBlend {blendOption=Linear}
    boundBone {boneName=dyn_coat_out_01_l}
    componentLinearAccScale=<1 1 1>
    componentLinearVelScale=<1 1 1>
    physicsBodyDefinitions [{
      sphereCollisionRadius=0
      boundBone {boneName=dyn_coat_out_01_l}
      boxExtents=<20 20 10>
      constraintSetup {angularTargetAxis=AxisY, twistAxis=AxisZ, angularLimitsMax=<30 30 0>, angularLimitsMin=<-30 -30 0>, angularTarget=<0 0 0>}
      localJointOffset=<0 0 -1>
     }]
   }
   n3 = Anim Dynamics - Bone: dyn_coat_fr_01_l(ComponentPose=@n2, Alpha=0.9, ExternalForce="0,0,0") {
    angularDampingOverride=1
    angularSpringConstant=2000
    bAngularSpring=true
    bGravityOverrideInSimSpace=true
    gravityScale=0.5
    lODThreshold=1
    linearDampingOverride=1
    linearSpringConstant=200
    alphaBoolBlend {blendOption=Linear}
    boundBone {boneName=dyn_coat_fr_01_l}
    componentLinearAccScale=<1 1 1>
    componentLinearVelScale=<1 1 1>
    physicsBodyDefinitions [{
      sphereCollisionRadius=0
      boundBone {boneName=dyn_coat_fr_01_l}
      boxExtents=<20 20 10>
      constraintSetup {angularTargetAxis=AxisY, twistAxis=AxisZ, angularLimitsMax=<30 30 0>, angularLimitsMin=<-30 -30 0>, angularTarget=<0 0 0>}
      localJointOffset=<0 0 -1>
     }]
   }
   n4 = Anim Dynamics - Bone: dyn_coat_bk_01_l(ComponentPose=@n3, Alpha=0.94, ExternalForce="0,0,0") {
    angularDampingOverride=1
    angularSpringConstant=2000
    bAngularSpring=true
    bGravityOverrideInSimSpace=true
    gravityScale=0.5
    lODThreshold=1
    linearDampingOverride=1
    linearSpringConstant=200
    alphaBoolBlend {blendOption=Linear}
    boundBone {boneName=dyn_coat_bk_01_l}
    componentLinearAccScale=<1 1 1>
    componentLinearVelScale=<1 1 1>
    physicsBodyDefinitions [{
      sphereCollisionRadius=0
      boundBone {boneName=dyn_coat_bk_01_l}
      boxExtents=<20 20 10>
      constraintSetup {angularTargetAxis=AxisY, twistAxis=AxisZ, angularLimitsMax=<0 30 30>, angularLimitsMin=<0 -30 -30>, angularTarget=<0 0 0>}
      localJointOffset=<0 0 -1>
     }]
   }
   n5 = Anim Dynamics - Bone: dyn_coat_frdiag_01_r(ComponentPose=@n4, Alpha=0.96, ExternalForce="0,0,0") {
    angularDampingOverride=1
    angularSpringConstant=2000
    bAngularSpring=true
    bGravityOverrideInSimSpace=true
    gravityScale=0.5
    lODThreshold=1
    linearDampingOverride=1
    linearSpringConstant=200
    alphaBoolBlend {blendOption=Linear}
    boundBone {boneName=dyn_coat_frdiag_01_r}
    componentLinearAccScale=<0.1 0.1 0.1>
    componentLinearVelScale=<0.1 0.1 0.1>
    physicsBodyDefinitions [{
      sphereCollisionRadius=0
      boundBone {boneName=dyn_coat_frdiag_01_r}
      boxExtents=<20 20 10>
      constraintSetup {angularTargetAxis=AxisY, twistAxis=AxisZ, angularLimitsMax=<30 30 0>, angularLimitsMin=<-30 -30 0>, angularTarget=<0 0 0>}
      localJointOffset=<0 0 -1>
     }]
   }
   n6 = Anim Dynamics - Bone: dyn_coat_fr_01_r(ComponentPose=@n5, Alpha=0.92, ExternalForce="0,0,0") {
    angularDampingOverride=1
    angularSpringConstant=2000
    bAngularSpring=true
    bGravityOverrideInSimSpace=true
    gravityScale=0.5
    lODThreshold=1
    linearDampingOverride=1
    linearSpringConstant=200
    alphaBoolBlend {blendOption=Linear}
    boundBone {boneName=dyn_coat_fr_01_r}
    componentLinearAccScale=<1 1 1>
    componentLinearVelScale=<1 1 1>
    physicsBodyDefinitions [{
      sphereCollisionRadius=0
      boundBone {boneName=dyn_coat_fr_01_r}
      boxExtents=<20 20 10>
      constraintSetup {angularTargetAxis=AxisY, twistAxis=AxisZ, angularLimitsMax=<30 45 0>, angularLimitsMin=<-30 -45 0>, angularTarget=<0 0 0>}
      localJointOffset=<0 0 -1>
     }]
   }
   Anim Dynamics - Bone: dyn_coat_bk_01_r(ComponentPose=@n6, Alpha=0.98, ExternalForce="0,0,0") {
    angularDampingOverride=1
    angularSpringConstant=2000
    bAngularSpring=true
    bGravityOverrideInSimSpace=true
    gravityScale=0.5
    lODThreshold=1
    linearDampingOverride=1
    linearSpringConstant=200
    alphaBoolBlend {blendOption=Linear}
    boundBone {boneName=dyn_coat_bk_01_r}
    componentLinearAccScale=<0.1 0.1 0.1>
    componentLinearVelScale=<0.1 0.1 0.1>
    physicsBodyDefinitions [{
      sphereCollisionRadius=0
      boundBone {boneName=dyn_coat_bk_01_r}
      boxExtents=<20 20 10>
      constraintSetup {angularTargetAxis=AxisY, twistAxis=AxisZ, angularLimitsMax=<15 30 0>, angularLimitsMin=<-15 -5 0>, angularTarget=<0 0 0>}
      localJointOffset=<0 0 -1>
     }]
   }
   n7 = Anim Dynamics - Bone: dyn_outer_collar_l(Alpha=0.125, ExternalForce="0,0,0") {
    angularDampingOverride=1
    angularSpringConstant=500
    bAngularSpring=true
    bGravityOverrideInSimSpace=true
    bOverrideAngularDamping=true
    bOverrideLinearDamping=true
    lODThreshold=1
    linearDampingOverride=1
    linearSpringConstant=125
    simulationSpace=Actor
    alphaBoolBlend {blendOption=Linear}
    boundBone {boneName=dyn_outer_collar_l}
    componentLinearAccScale=<1 1 1>
    componentLinearVelScale=<1 1 1>
    physicsBodyDefinitions [{
      sphereCollisionRadius=0
      boundBone {boneName=dyn_outer_collar_l}
      boxExtents=<30 30 30>
      constraintSetup {angularTargetAxis=AxisY, angularLimitsMax=<0 15 15>, angularLimitsMin=<0 -15 -15>, angularTarget=<0 0 0>}
      localJointOffset=<0 0 0>
     }]
   }
   n8 = Anim Dynamics - Bone: dyn_outer_collar_r(ComponentPose=@n7, Alpha=0.125, ExternalForce="0,0,0") {
    angularDampingOverride=1
    angularSpringConstant=500
    bAngularSpring=true
    bGravityOverrideInSimSpace=true
    bOverrideAngularDamping=true
    bOverrideLinearDamping=true
    lODThreshold=1
    linearDampingOverride=1
    linearSpringConstant=125
    simulationSpace=Actor
    alphaBoolBlend {blendOption=Linear}
    boundBone {boneName=dyn_outer_collar_r}
    componentLinearAccScale=<1 1 1>
    componentLinearVelScale=<1 1 1>
    physicsBodyDefinitions [{
      sphereCollisionRadius=0
      boundBone {boneName=dyn_outer_collar_r}
      boxExtents=<30 30 30>
      constraintSetup {angularTargetAxis=AxisY, angularLimitsMax=<0 15 15>, angularLimitsMin=<0 -15 -15>, angularTarget=<0 0 0>}
      localJointOffset=<0 0 0>
     }]
   }
   n9 = Anim Dynamics - Bone: dyn_shoulderpad_l(ComponentPose=@n8, Alpha=0.125, ExternalForce="0,0,0") {
    angularDampingOverride=0.98
    angularSpringConstant=40
    bAngularSpring=true
    bGravityOverrideInSimSpace=true
    bOverrideAngularDamping=true
    bOverrideLinearDamping=true
    gravityScale=0
    lODThreshold=1
    linearDampingOverride=0.98
    linearSpringConstant=40
    simulationSpace=Actor
    alphaBoolBlend {blendOption=Linear}
    boundBone {boneName=dyn_shoulderpad_l}
    componentLinearAccScale=<0.25 0.25 0.25>
    componentLinearVelScale=<0.25 0.25 0.25>
    physicsBodyDefinitions [{
      sphereCollisionRadius=0
      boundBone {boneName=dyn_shoulderpad_l}
      boxExtents=<30 30 30>
      constraintSetup {angularTargetAxis=AxisY, angularLimitsMax=<0 30 30>, angularLimitsMin=<0 -30 -30>, angularTarget=<0 0 0>}
      localJointOffset=<0 0 0>
     }]
   }
   Anim Dynamics - Bone: dyn_shoulderpad_r(ComponentPose=@n9, Alpha=0.125, ExternalForce="0,0,0") {
    angularDampingOverride=0.98
    angularSpringConstant=40
    bAngularSpring=true
    bGravityOverrideInSimSpace=true
    bOverrideAngularDamping=true
    bOverrideLinearDamping=true
    gravityScale=0
    lODThreshold=1
    linearDampingOverride=0.98
    linearSpringConstant=40
    simulationSpace=Actor
    alphaBoolBlend {blendOption=Linear}
    boundBone {boneName=dyn_shoulderpad_r}
    componentLinearAccScale=<0.25 0.25 0.25>
    componentLinearVelScale=<0.25 0.25 0.25>
    physicsBodyDefinitions [{
      sphereCollisionRadius=0
      boundBone {boneName=dyn_shoulderpad_r}
      boxExtents=<30 30 30>
      constraintSetup {angularTargetAxis=AxisY, angularLimitsMax=<0 30 30>, angularLimitsMin=<0 -30 -30>, angularTarget=<0 0 0>}
      localJointOffset=<0 0 0>
     }]
   }
   n10 = RigidBody(ComponentPose=Local To Component(LocalPose=Use cached pose 'CrouchCache'())) {
    bClampLinearTranslationLimitToRefPose=true
    bOverrideWorldGravity=true
    overridePhysicsAsset=&/M_MED_Werewolf_02_Body_secondary_Physics
    simulationSpace=BaseBoneSpace
    alphaBoolBlend {blendOption=Linear}
    alphaScaleBiasClamp {bInterpResult=true, interpSpeedDecreasing=6}
    baseBoneRef {boneName=root}
    componentAppliedLinearAccClamp=<6000 6000 6000>
    componentLinearAccScale=<0.25 0.25 0.25>
    componentLinearVelScale=<0.25 0.25 0.25>
    overrideWorldGravity {z=-980}
   }
   n11 = Component To Local(ComponentPose=@n10)
   n12 = RigidBody(ComponentPose=Local To Component(LocalPose=Use cached pose 'CrouchCache'()), Alpha=0.9) {
    bClampLinearTranslationLimitToRefPose=true
    bOverrideWorldGravity=true
    overridePhysicsAsset=&/M_MED_Werewolf_02_Body_Physics
    simulationSpace=BaseBoneSpace
    alphaBoolBlend {blendOption=Linear}
    alphaScaleBiasClamp {bInterpResult=true, interpSpeedDecreasing=6}
    baseBoneRef {boneName=root}
    componentAppliedLinearAccClamp=<6000 6000 6000>
    componentLinearAccScale=<0.25 0.25 0.25>
    componentLinearVelScale=<0.25 0.25 0.25>
    overrideWorldGravity {z=-980}
   }
   n13 = Component To Local(ComponentPose=@n12)
   n14 = Get IsCrouching() -> (IsCrouching:bool)
   n15 = "Transform (Modify) Bone - Bone: neck_02"(
    ComponentPose=Local To Component(LocalPose=Blend Poses by bool(BlendPose_0=@n11, BlendPose_1=@n13, BlendTime_0=1, BlendTime_1=1.5, bActiveValue=@n14) {blendType=Linear})
    Alpha=1
    Translation="1.1,0,0"
    Rotation="0,0,0"
    Scale="0.9,0.9,0.9"
   ) {
    scaleMode=BMM_Additive
    scaleSpace=BCS_BoneSpace
    translationMode=BMM_Additive
    translationSpace=BCS_BoneSpace
    alphaBoolBlend {blendOption=Linear}
    boneToModify {boneName=neck_02}
    scale=<0.9 0.9 0.9>
    translation {x=1.1}
   }
   Output Pose(Result=Component To Local(ComponentPose=@n15)) {layerGroup=None}
   n16 = Control Rig(Source=Copy Pose From Mesh() {bUseAttachedParent=true}) {
    alphaBoolBlend {blendOption=Linear}
    controlRigAssetReference {blueprintRigClass=&/M_MED_Werewolf_ControlRig.M_MED_Werewolf_ControlRig_C}
   }
   n17 = Anim Dynamics - Bone: dyn_coat_bk_01(ComponentPose=Local To Component(LocalPose=@n16), Alpha=1, ExternalForce="0,0,0") {
    angularDampingOverride=0.9999
    angularSpringConstant=2000
    bAngularSpring=true
    bGravityOverrideInSimSpace=true
    bOverrideAngularDamping=true
    gravityScale=0.5
    lODThreshold=1
    linearDampingOverride=1
    linearSpringConstant=200
    simulationSpace=RootRelative
    alphaBoolBlend {blendOption=Linear}
    boundBone {boneName=dyn_coat_bk_01}
    componentAppliedLinearAccClamp=<9000 9000 9000>
    componentLinearAccScale=<1 1 1>
    componentLinearVelScale=<1 1 1>
    physicsBodyDefinitions [{
      sphereCollisionRadius=0
      boundBone {boneName=dyn_coat_bk_01}
      boxExtents=<10 10 10>
      constraintSetup {angularTargetAxis=AxisY, twistAxis=AxisZ, angularLimitsMax=<30 30 0>, angularLimitsMin=<-30 -30 0>, angularTarget=<1 0 0>}
      localJointOffset=<0 0 -1>
     }]
   }
   n18 = Anim Dynamics - Bone: dyn_coat_out_01_l(ComponentPose=@n17, Alpha=1, ExternalForce="0,0,0") {
    angularDampingOverride=0.9999
    angularSpringConstant=2000
    bAngularSpring=true
    bGravityOverrideInSimSpace=true
    bOverrideAngularDamping=true
    gravityScale=0.5
    lODThreshold=1
    linearDampingOverride=1
    linearSpringConstant=200
    simulationSpace=RootRelative
    alphaBoolBlend {blendOption=Linear}
    boundBone {boneName=dyn_coat_out_01_l}
    componentAppliedLinearAccClamp=<9000 9000 9000>
    componentLinearAccScale=<1 1 1>
    componentLinearVelScale=<1 1 1>
    physicsBodyDefinitions [{
      sphereCollisionRadius=0
      boundBone {boneName=dyn_coat_out_01_l}
      boxExtents=<10 10 10>
      constraintSetup {angularTargetAxis=AxisY, twistAxis=AxisZ, angularLimitsMax=<30 10 0>, angularLimitsMin=<-30 -30 0>, angularTarget=<1 0 0>}
      localJointOffset=<0 0 -1>
     }]
   }
   n19 = Anim Dynamics - Bone: dyn_coat_fr_01_l(ComponentPose=@n18, Alpha=1, ExternalForce="0,0,0") {
    angularDampingOverride=0.9999
    angularSpringConstant=2000
    bAngularSpring=true
    bGravityOverrideInSimSpace=true
    bOverrideAngularDamping=true
    gravityScale=0.5
    lODThreshold=1
    linearDampingOverride=1
    linearSpringConstant=200
    simulationSpace=RootRelative
    alphaBoolBlend {blendOption=Linear}
    boundBone {boneName=dyn_coat_fr_01_l}
    componentAppliedLinearAccClamp=<9000 9000 9000>
    componentLinearAccScale=<1 1 1>
    componentLinearVelScale=<1 1 1>
    physicsBodyDefinitions [{
      sphereCollisionRadius=0
      boundBone {boneName=dyn_coat_fr_01_l}
      boxExtents=<10 10 10>
      constraintSetup {angularTargetAxis=AxisY, twistAxis=AxisZ, angularLimitsMax=<10 15 0>, angularLimitsMin=<-30 -15 0>, angularTarget=<1 0 0>}
      localJointOffset=<0 0 -1>
     }]
   }
   n20 = Anim Dynamics - Bone: dyn_coat_bk_01_l(ComponentPose=@n19, Alpha=1, ExternalForce="0,0,0") {
    angularDampingOverride=0.9999
    angularSpringConstant=2000
    bAngularSpring=true
    bGravityOverrideInSimSpace=true
    bOverrideAngularDamping=true
    gravityScale=0.5
    lODThreshold=1
    linearDampingOverride=1
    linearSpringConstant=200
    simulationSpace=RootRelative
    alphaBoolBlend {blendOption=Linear}
    boundBone {boneName=dyn_coat_bk_01_l}
    componentAppliedLinearAccClamp=<9000 9000 9000>
    componentLinearAccScale=<1 1 1>
    componentLinearVelScale=<1 1 1>
    physicsBodyDefinitions [{
      sphereCollisionRadius=0
      boundBone {boneName=dyn_coat_bk_01_l}
      boxExtents=<10 10 10>
      constraintSetup {angularTargetAxis=AxisY, twistAxis=AxisZ, angularLimitsMax=<0 10 20>, angularLimitsMin=<0 -30 -20>, angularTarget=<1 0 0>}
      localJointOffset=<0 0 -1>
     }]
   }
   n21 = Anim Dynamics - Bone: dyn_coat_frdiag_01_r(ComponentPose=@n20, Alpha=1, ExternalForce="0,0,0") {
    angularDampingOverride=0.9999
    angularSpringConstant=2000
    bAngularSpring=true
    bGravityOverrideInSimSpace=true
    bOverrideAngularDamping=true
    gravityScale=0.5
    lODThreshold=1
    linearDampingOverride=1
    linearSpringConstant=200
    simulationSpace=RootRelative
    alphaBoolBlend {blendOption=Linear}
    boundBone {boneName=dyn_coat_frdiag_01_r}
    componentAppliedLinearAccClamp=<9000 9000 9000>
    componentLinearAccScale=<1 1 1>
    componentLinearVelScale=<1 1 1>
    physicsBodyDefinitions [{
      sphereCollisionRadius=0
      boundBone {boneName=dyn_coat_frdiag_01_r}
      boxExtents=<10 10 10>
      constraintSetup {angularTargetAxis=AxisY, twistAxis=AxisZ, angularLimitsMax=<30 30 0>, angularLimitsMin=<-30 -10 0>, angularTarget=<1 0 0>}
      localJointOffset=<0 0 -1>
     }]
   }
   n22 = Anim Dynamics - Bone: dyn_coat_fr_01_r(ComponentPose=@n21, Alpha=1, ExternalForce="0,0,0") {
    angularDampingOverride=0.9999
    angularSpringConstant=2000
    bAngularSpring=true
    bGravityOverrideInSimSpace=true
    bOverrideAngularDamping=true
    gravityScale=0.5
    lODThreshold=1
    linearDampingOverride=1
    linearSpringConstant=200
    simulationSpace=RootRelative
    alphaBoolBlend {blendOption=Linear}
    boundBone {boneName=dyn_coat_fr_01_r}
    componentAppliedLinearAccClamp=<9000 9000 9000>
    componentLinearAccScale=<1 1 1>
    componentLinearVelScale=<1 1 1>
    physicsBodyDefinitions [{
      sphereCollisionRadius=0
      boundBone {boneName=dyn_coat_fr_01_r}
      boxExtents=<10 10 10>
      constraintSetup {angularTargetAxis=AxisY, twistAxis=AxisZ, angularLimitsMax=<10 30 0>, angularLimitsMin=<-30 -30 0>, angularTarget=<1 0 0>}
      localJointOffset=<0 0 -1>
     }]
   }
   n23 = Anim Dynamics - Bone: dyn_coat_bk_01_r(ComponentPose=@n22, Alpha=1, ExternalForce="0,0,0") {
    angularDampingOverride=0.9999
    angularSpringConstant=2000
    bAngularSpring=true
    bGravityOverrideInSimSpace=true
    bOverrideAngularDamping=true
    gravityScale=0.5
    lODThreshold=1
    linearDampingOverride=1
    linearSpringConstant=200
    simulationSpace=RootRelative
    alphaBoolBlend {blendOption=Linear}
    boundBone {boneName=dyn_coat_bk_01_r}
    componentAppliedLinearAccClamp=<9000 9000 9000>
    componentLinearAccScale=<1 1 1>
    componentLinearVelScale=<1 1 1>
    physicsBodyDefinitions [{
      sphereCollisionRadius=0
      boundBone {boneName=dyn_coat_bk_01_r}
      boxExtents=<10 10 10>
      constraintSetup {angularTargetAxis=AxisY, twistAxis=AxisZ, angularLimitsMax=<15 30 0>, angularLimitsMin=<-5 -5 0>, angularTarget=<1 0 0>}
      localJointOffset=<0 0 -1>
     }]
   }
   n24 = Anim Dynamics - Bone: dyn_outer_collar_l(ComponentPose=@n23, Alpha=0.125, ExternalForce="0,0,0") {
    angularDampingOverride=1
    angularSpringConstant=500
    bAngularSpring=true
    bGravityOverrideInSimSpace=true
    bLinearSpring=true
    bOverrideAngularDamping=true
    bOverrideLinearDamping=true
    lODThreshold=1
    linearDampingOverride=1
    linearSpringConstant=125
    simulationSpace=RootRelative
    alphaBoolBlend {blendOption=Linear}
    boundBone {boneName=dyn_outer_collar_l}
    componentLinearAccScale=<1 1 1>
    componentLinearVelScale=<1 1 1>
    physicsBodyDefinitions [{
      sphereCollisionRadius=0
      boundBone {boneName=dyn_outer_collar_l}
      boxExtents=<30 30 30>
      constraintSetup {
       angularTargetAxis=AxisY
       angularLimitsMax=<0 15 15>
       angularLimitsMin=<0 -15 -15>
       angularTarget=<0 0 0>
       linearAxesMax=<10 10 10>
       linearAxesMin=<0 -10 -10>
      }
      localJointOffset=<0 0 0>
     }]
   }
   n25 = Anim Dynamics - Bone: dyn_outer_collar_r(ComponentPose=@n24, Alpha=0.125, ExternalForce="0,0,0") {
    angularDampingOverride=1
    angularSpringConstant=500
    bAngularSpring=true
    bGravityOverrideInSimSpace=true
    bLinearSpring=true
    bOverrideAngularDamping=true
    bOverrideLinearDamping=true
    lODThreshold=1
    linearDampingOverride=1
    linearSpringConstant=125
    simulationSpace=RootRelative
    alphaBoolBlend {blendOption=Linear}
    boundBone {boneName=dyn_outer_collar_r}
    componentLinearAccScale=<1 1 1>
    componentLinearVelScale=<1 1 1>
    physicsBodyDefinitions [{
      sphereCollisionRadius=0
      boundBone {boneName=dyn_outer_collar_r}
      boxExtents=<30 30 30>
      constraintSetup {
       angularTargetAxis=AxisY
       angularLimitsMax=<0 15 15>
       angularLimitsMin=<0 -15 -15>
       angularTarget=<0 0 0>
       linearAxesMax=<10 10 10>
       linearAxesMin=<0 -10 -10>
      }
      localJointOffset=<0 0 0>
     }]
   }
   n26 = Anim Dynamics - Bone: dyn_shoulderpad_l(ComponentPose=@n25, Alpha=0.125, ExternalForce="0,0,0") {
    angularDampingOverride=0.98
    angularSpringConstant=40
    bAngularSpring=true
    bGravityOverrideInSimSpace=true
    bLinearSpring=true
    bOverrideAngularDamping=true
    bOverrideLinearDamping=true
    gravityScale=0
    lODThreshold=1
    linearDampingOverride=0.98
    linearSpringConstant=40
    simulationSpace=RootRelative
    alphaBoolBlend {blendOption=Linear}
    boundBone {boneName=dyn_shoulderpad_l}
    componentLinearAccScale=<0.25 0.25 0.25>
    componentLinearVelScale=<0.25 0.25 0.25>
    physicsBodyDefinitions [{
      sphereCollisionRadius=0
      boundBone {boneName=dyn_shoulderpad_l}
      boxExtents=<30 30 30>
      constraintSetup {
       angularTargetAxis=AxisY
       angularLimitsMax=<0 30 30>
       angularLimitsMin=<0 -30 -30>
       angularTarget=<0 0 0>
       linearAxesMax=<10 20 20>
       linearAxesMin=<0 -20 -20>
      }
      localJointOffset=<0 0 0>
     }]
   }
   n27 = Anim Dynamics - Bone: dyn_shoulderpad_r(ComponentPose=@n26, Alpha=0.125, ExternalForce="0,0,0") {
    angularDampingOverride=0.98
    angularSpringConstant=40
    bAngularSpring=true
    bGravityOverrideInSimSpace=true
    bLinearSpring=true
    bOverrideAngularDamping=true
    bOverrideLinearDamping=true
    gravityScale=0
    lODThreshold=1
    linearDampingOverride=0.98
    linearSpringConstant=40
    simulationSpace=RootRelative
    alphaBoolBlend {blendOption=Linear}
    boundBone {boneName=dyn_shoulderpad_r}
    componentLinearAccScale=<0.25 0.25 0.25>
    componentLinearVelScale=<0.25 0.25 0.25>
    physicsBodyDefinitions [{
      sphereCollisionRadius=0
      boundBone {boneName=dyn_shoulderpad_r}
      boxExtents=<30 30 30>
      constraintSetup {
       angularTargetAxis=AxisY
       angularLimitsMax=<0 30 30>
       angularLimitsMin=<0 -30 -30>
       angularTarget=<0 0 0>
       linearAxesMax=<10 20 20>
       linearAxesMin=<0 -20 -20>
      }
      localJointOffset=<0 0 0>
     }]
   }
   n28 = "Constraint - dyn_mane_04 (P)"(ComponentPose=@n27, Alpha=1, ConstraintWeights_0=1) {
    alphaBoolBlend {blendOption=Linear}
    boneToModify {boneName=dyn_mane_04}
    constraintSetup [{targetBone {boneName=neck_01}}]
    constraintWeights [1]
   }
   n29 = "Constraint - dyn_mane_05 (PP)"!(ComponentPose=@n28, Alpha=1, ConstraintWeights_0=0.5, ConstraintWeights_1:float=0.5) -> (Pose) {
    alphaBoolBlend {blendOption=Linear}
    alphaScaleBiasClamp {scale=0.5}
    boneToModify {boneName=dyn_mane_05}
    constraintSetup [{targetBone {boneName=neck_01}}, {targetBone {boneName=spine_05}}]
    constraintWeights [1,1]
   }
   Save cached pose 'CrouchCache'(Pose=Component To Local(ComponentPose=@n29)) {cacheName=CrouchCache}
  ]}
 EventGraph {nodes [
   Event BlueprintUpdateAnimation()
   n1 = Get PawnSpeed2D()
   > Set IsCrouching(IsCrouching:bool=AND Boolean(A=Get bIsCrouching(), B="Equal (Float)"!(A:double=@n1, B:double=0) -> (ReturnValue:bool) {is=title K2Node_CallFunction})) -> (Output_Get:bool)
  ]}
]
variables [IsCrouching {type=boolean}]
