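# Physics asset of cape colliders for the medium male player werewolf body M_MED_Werewolf_02, presumably for cape cloth collision, going by its name.
# Capsules approximate clavicles, upper chest, arms and the wolf-specific thigh-twist, calf and foot bones, while zero-length capsules act as spheres on head and pelvis.
# It defines no constraints, so the bodies form a collision proxy on the bones rather than a jointed ragdoll.
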PhysicsAsset /Game/Characters/Player/Male/Medium/Bodies/M_MED_Werewolf_01/Meshes/M_MED_Werewolf_02_Body_CapeColliders
bodies {
 clavicle_l [{shapeType=Capsule, length=10.3261, radius=12.0802, center=<8.2823 -4.7149 -2.0684>, rotation=<rot 90 -0.0025 179.9976>}]
 clavicle_r [{shapeType=Capsule, length=10.3261, radius=12.0802, center=<-8.4184 5.0596 2.5747>, rotation=<rot 90 0.0099 0.0099>}]
 head [{shapeType=Capsule, length=0, radius=18, center=<3.4835 2 0>}]
 lowerarm_l [{shapeType=Capsule, length=15, radius=8.25, center=<14.2109 0.0001 0>, rotation=<rot -90 -2.5305 2.5305>}]
 lowerarm_r [{shapeType=Capsule, length=15, radius=8.25, center=<-14.6692 0.0001 0.0004>, rotation=<rot -90 -0.0264 0.0264>}]
 pelvis [{shapeType=Capsule, length=0, radius=13, center=<0 2.8244 0>}]
 spine_03 [{shapeType=Capsule, length=10.5169, radius=16.5072, center=<-1.5797 -0.2471 0.2523>, rotation=<rot -83.7243 -83.7209 89.9955>}]
 upperarm_l [{shapeType=Capsule, length=15, radius=7.5, center=<14.0805 -0.0001 0.4948>, rotation=<rot -90 2.5305 -2.5304>}]
 upperarm_r [{shapeType=Capsule, length=15, radius=7.5, center=<-11.7411 0 0.4948>, rotation=<rot -90 -3.7927 3.7927>}]
 wolf_calf_l [{shapeType=Capsule, length=20.25, radius=7, center=<-9.6325 -5.0724 -0.0001>, rotation=<rot 90 159.444 134.4393>}]
 wolf_calf_r [{shapeType=Capsule, length=20.25, radius=7, center=<11.9212 4.9812 -0.0001>, rotation=<rot 90 0 -25.0037>}]
 wolf_foot_l [{shapeType=Capsule, length=20, radius=6.5, center=<-10.6797 1.8905 -0.9352>, rotation=<rot -85.0011 0 -10.0017>}]
 wolf_foot_r [{shapeType=Capsule, length=20, radius=6.5, center=<11.4658 -2.0296 1.0032>, rotation=<rot -85.0002 0 -10.0008>}]
 wolf_thigh_twist_01_l [{shapeType=Capsule, length=31, radius=9.25, center=<0 -0.6548 -0.0002>, rotation=<rot 90 0.0049 0.0049>}]
 wolf_thigh_twist_01_r [{shapeType=Capsule, length=31, radius=9.25, center=<0 -0.6548 -0.0002>, rotation=<rot 90 0.0049 0.0049>}]
}
constraints []
