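# Physics asset for the Werewolf 02 male medium character's face accessory, giving bodies to left and right bandanna tails.
# Each tail is a four-capsule chain hung from a small master body linked to the head, with a tighter angular limit at the root than along the chain.
# Head, upper spine, clavicle and upper-arm capsules form the body volumes; their constraints are free, and the left clavicle and arm have none.
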PhysicsAsset /Game/Characters/Player/Male/Medium/Bodies/M_MED_Werewolf_01/Meshes/M_MED_Werewolf_02_FaceAcc_Physics
bodies {
 clavicle_l [{shapeType=Capsule, length=9.375, radius=7.5, center=<9.7252 0 0>, rotation=<rot -90 0 0>}]
 clavicle_r [{shapeType=Capsule, length=9.375, radius=7.5, center=<-9.7252 0 0>, rotation=<rot -90 0 0>}]
 dyn_bandanna_l_01 [{shapeType=Capsule, length=5, radius=2, shapeName=dyn_bandanna_l_01_capsule, center=<6.7795 0 0>, rotation=<rot 90 0 0>}]
 dyn_bandanna_l_03 [{shapeType=Capsule, length=10, radius=2, shapeName=dyn_bandanna_l_03_capsule, center=<3.315 0 0>, rotation=<rot 90 0 0>}]
 dyn_bandanna_l_05 [{shapeType=Capsule, length=12.6242, radius=2, shapeName=dyn_bandanna_l_05_capsule, center=<6.7341 0 0>, rotation=<rot 90 0 0>}]
 dyn_bandanna_l_07 [{shapeType=Capsule, length=10, radius=2, shapeName=dyn_bandanna_l_07_capsule, center=<1.9216 0 0>, rotation=<rot 90 0 0>}]
 dyn_bandanna_l_mstr [{shapeType=Capsule, length=0.505, radius=0.505, shapeName=dyn_bandanna_l_mstr_capsule}]
 dyn_bandanna_r_01 [{shapeType=Capsule, length=5, radius=2, shapeName=dyn_bandanna_l_01_capsule, center=<6.7795 0 0>, rotation=<rot 90 0 0>}]
 dyn_bandanna_r_02 [{shapeType=Capsule, length=10, radius=2, shapeName=dyn_bandanna_l_03_capsule, center=<3.315 0 0>, rotation=<rot 90 0 0>}]
 dyn_bandanna_r_03 [{shapeType=Capsule, length=10, radius=2, shapeName=dyn_bandanna_l_05_capsule, center=<5.2816 0 0>, rotation=<rot 90 0 0>}]
 dyn_bandanna_r_04 [{shapeType=Capsule, length=10, radius=2, shapeName=dyn_bandanna_l_07_capsule, center=<1.9216 0 0>, rotation=<rot 90 0 0>}]
 dyn_bandanna_r_mstr [{shapeType=Capsule, length=0.505, radius=0.505, shapeName=dyn_bandanna_l_mstr_capsule}]
 head [{shapeType=Capsule, length=0, radius=11.875, center=<-2.8021 0.8878 -0.0001>}]
 spine_05 [
  {shapeType=Capsule, length=3, radius=18.5213, center=<1.8814 -0.8866 -0.0001>}
  {shapeType=Capsule, length=13.7902, radius=6, shapeName=spine_05_capsule, center=<16.8767 -14.6512 0>, rotation=<rot 90 29.9999 0>}
 ]
 upperarm_l [{shapeType=Capsule, length=27.375, radius=7, center=<13.1604 0 0>, rotation=<rot -88.1986 -86.898 89.6603>}]
 upperarm_r [{shapeType=Capsule, length=27.375, radius=7, center=<-13.1604 0 0>, rotation=<rot -88.1986 -86.898 89.6603>}]
}
constraints [
 {bone1Name=dyn_bandanna_l_03, bone2Name=dyn_bandanna_l_01, swing1LimitDegrees=45, swing2LimitDegrees=45, twistLimitDegrees=45}
 {bone1Name=dyn_bandanna_l_05, bone2Name=dyn_bandanna_l_03, swing1LimitDegrees=45, swing2LimitDegrees=45, twistLimitDegrees=45}
 {bone1Name=dyn_bandanna_l_07, bone2Name=dyn_bandanna_l_05, swing1LimitDegrees=45, swing2LimitDegrees=45, twistLimitDegrees=45}
 {bone1Name=dyn_bandanna_l_01, bone2Name=dyn_bandanna_l_mstr, swing1LimitDegrees=20, swing2LimitDegrees=20, twistLimitDegrees=20}
 {
  bone1Name=head
  bone2Name=spine_05
  swing1LimitDegrees=45
  swing1Motion=Free
  swing2LimitDegrees=45
  swing2Motion=Free
  twistLimitDegrees=45
  twistMotion=Free
 }
 {
  bone1Name=dyn_bandanna_l_mstr
  bone2Name=head
  swing1LimitDegrees=45
  swing1Motion=Free
  swing2LimitDegrees=45
  swing2Motion=Free
  twistLimitDegrees=45
  twistMotion=Free
 }
 {
  bone1Name=dyn_bandanna_r_mstr
  bone2Name=head
  swing1LimitDegrees=45
  swing1Motion=Free
  swing2LimitDegrees=45
  swing2Motion=Free
  twistLimitDegrees=45
  twistMotion=Free
 }
 {bone1Name=dyn_bandanna_r_01, bone2Name=dyn_bandanna_r_mstr, swing1LimitDegrees=20, swing2LimitDegrees=20, twistLimitDegrees=20}
 {bone1Name=dyn_bandanna_r_02, bone2Name=dyn_bandanna_r_01, swing1LimitDegrees=45, swing2LimitDegrees=45, twistLimitDegrees=45}
 {bone1Name=dyn_bandanna_r_03, bone2Name=dyn_bandanna_r_02, swing1LimitDegrees=45, swing2LimitDegrees=45, twistLimitDegrees=45}
 {bone1Name=dyn_bandanna_r_04, bone2Name=dyn_bandanna_r_03, swing1LimitDegrees=45, swing2LimitDegrees=45, twistLimitDegrees=45}
 {
  bone1Name=upperarm_r
  bone2Name=clavicle_r
  swing1LimitDegrees=45
  swing1Motion=Free
  swing2LimitDegrees=45
  swing2Motion=Free
  twistLimitDegrees=45
  twistMotion=Free
 }
 {
  bone1Name=clavicle_r
  bone2Name=spine_05
  swing1LimitDegrees=45
  swing1Motion=Free
  swing2LimitDegrees=45
  swing2Motion=Free
  twistLimitDegrees=45
  twistMotion=Free
 }
]
