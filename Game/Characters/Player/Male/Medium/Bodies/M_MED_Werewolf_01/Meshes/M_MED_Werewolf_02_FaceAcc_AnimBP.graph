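# Animation Blueprint for the male medium Werewolf 02 player face accessory, driving its bandanna bones with attached-parent motion and physics.
# It layers two copied mesh poses at the bandanna and upper-arm bones, one sourced from a skeletal mesh cached at initialization.
# Weighted constraints tie both bandanna master bones to faceAttach and the neck, feeding a RigidBody simulation in spine_03 space with an override physics asset.
# A bake-pose sequence player is left unconnected.
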AnimBlueprint /Game/Characters/Player/Male/Medium/Bodies/M_MED_Werewolf_01/Meshes/M_MED_Werewolf_02_FaceAcc_AnimBP
target_skeleton=&/M_MED_Werewolf_02_FaceAcc_Skeleton
graphs [
 AnimGraph {nodes [
   n1 = Layered blend per bone(
    BasePose=Copy Pose From Mesh() {bUseAttachedParent=true}
    BlendPoses_0=Copy Pose From Mesh(SourceMeshComponent=Get skelMesh() -> (skelMesh:Skeletal Mesh Component Object Reference)) {bUseAttachedParent=true}
   ) {layerSetup [{branchFilters [
       {blendDepth=1, boneName=dyn_bandanna_l_mstr}
       {blendDepth=1, boneName=dyn_bandanna_r_mstr}
       {blendDepth=1, boneName=upperarm_l}
       {blendDepth=1, boneName=upperarm_r}
      ]}]}
   n2 = "Constraint - dyn_bandanna_l_mstr (PPP)"!(
    ComponentPose=Local To Component(LocalPose=@n1)
    Alpha=1
    ConstraintWeights_0=0.5652
    ConstraintWeights_1:float=0.1132
    ConstraintWeights_2:float=0.3216
   ) -> (Pose) {
    alphaBoolBlend {blendOption=Linear}
    boneToModify {boneName=dyn_bandanna_l_mstr}
    constraintSetup [{targetBone {boneName=faceAttach}}, {targetBone {boneName=neck_01}}, {targetBone {boneName=neck_02}}]
    constraintWeights [1,1,1]
   }
   n3 = "Constraint - dyn_bandanna_r_mstr (PPP)"!(ComponentPose=@n2, Alpha=1, ConstraintWeights_0=0.5652, ConstraintWeights_1:float=0.1132, ConstraintWeights_2:float=0.3216) -> (Pose) {
    alphaBoolBlend {blendOption=Linear}
    boneToModify {boneName=dyn_bandanna_r_mstr}
    constraintSetup [{targetBone {boneName=faceAttach}}, {targetBone {boneName=neck_01}}, {targetBone {boneName=neck_02}}]
    constraintWeights [1,1,1]
   }
   n4 = RigidBody(ComponentPose=@n3) {
    bOverrideWorldGravity=true
    overridePhysicsAsset=&/M_MED_Werewolf_02_FaceAcc_Physics
    simulationSpace=BaseBoneSpace
    alphaBoolBlend {blendOption=Linear}
    baseBoneRef {boneName=spine_03}
    overrideWorldGravity {z=-980}
    simSpaceSettings {
     dampingAlpha=0
     maxAngularAcceleration=50
     maxAngularVelocity=7
     maxLinearAcceleration=4000
     maxLinearVelocity=800
     velocityScaleZ=0.5
     worldAlpha=1
     externalLinearDragV=<2 2 1.5>
    }
   }
   Output Pose(Result=Component To Local(ComponentPose=@n4)) {layerGroup=None}
   Sequence Player 'M_MED_Werewolf_02_FaceAcc_BakePose'() {bLoopAnimation=false, sequence=&/M_MED_Werewolf_02_FaceAcc_BakePose}
  ]}
 EventGraph {nodes [
   Event BlueprintInitializeAnimation()
   > Set skelMesh(skelMesh:Skeletal Mesh Component Object Reference=GetOwnerSkeletalMeshComponent()) -> (Output_Get:Skeletal Mesh Component Object Reference)
  ]}
]
variables [skelMesh {type=SkeletalMeshComponent}]
