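# Physics asset for the left-side dangling bandanna of the Werewolf 02 male player character's face accessory.
# Capsules on every other bandanna bone (01 to 07) form a constrained chain with tight twist and swing that mostly widens toward the tip.
# The root bone is constrained to both head and neck_02.
# Head, neck, upper spine, left clavicle and upper arm carry body shapes, including a large offset head sphere.
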PhysicsAsset /Game/Characters/Player/Male/Medium/Bodies/M_MED_Werewolf_01/Meshes/M_MED_Werewolf_02_FaceAcc_Physics_L
bodies {
 clavicle_l [{shapeType=Capsule, length=9.375, radius=7.5, center=<9.7252 0 0>, rotation=<rot -90 0.0003 -0.0002>}]
 dyn_bandanna_l_01 [{
   shapeType=Capsule
   length=4.1979
   radius=2.4359
   shapeName=dyn_bandanna_l_01_capsule
   center=<6.7795 -0.0825 0.3209>
   rotation=<rot 0 -95.8364 -89.5485>
  }]
 dyn_bandanna_l_03 [{
   shapeType=Capsule
   length=3.3219
   radius=3.8692
   shapeName=dyn_bandanna_l_03_capsule
   center=<3.315 0.0096 1.9196>
   rotation=<rot 0 -89.8695 -88.981>
  }]
 dyn_bandanna_l_05 [{
   shapeType=Capsule
   length=7.3301
   radius=4.6308
   shapeName=dyn_bandanna_l_05_capsule
   center=<5.2816 0.0559 1.4631>
   rotation=<rot 0 -88.3872 -89.1546>
  }]
 dyn_bandanna_l_07 [{
   shapeType=Capsule
   length=5.1303
   radius=5.806
   shapeName=dyn_bandanna_l_07_capsule
   center=<1.9216 0.0308 0.8114>
   rotation=<rot 0 89.6663 -88.9358>
  }]
 head [
  {shapeType=Sphere, radius=50.5, center=<58.428 17.3581 0.0002>}
  {shapeType=Capsule, length=0, radius=11.875, center=<-2.8021 0.8878 -0.0001>}
 ]
 neck_01 [{
   shapeType=Capsule
   length=13.555
   radius=5.0304
   shapeName=neck_01_capsule
   center=<-4.6125 -11.6463 2.8981>
   rotation=<rot 90 -178.8065 -169.9622>
  }]
 neck_02 [{shapeType=Capsule, length=12.513, radius=3.619, shapeName=neck_02_capsule, center=<8.3798 -9.1742 2.9268>, rotation=<rot 90 180 144.7141>}]
 spine_05 [{shapeType=Capsule, length=7.6988, radius=18.5213, center=<1.4461 2.4674 -0.0001>}]
 upperarm_l [{shapeType=Capsule, length=27.375, radius=7, center=<13.1604 0.6274 -0.2085>, rotation=<rot -88.1986 -86.898 89.6603>}]
}
constraints [
 {bone1Name=dyn_bandanna_l_03, bone2Name=dyn_bandanna_l_01, swing1LimitDegrees=8.76, swing2LimitDegrees=22.5, twistLimitDegrees=3}
 {bone1Name=dyn_bandanna_l_05, bone2Name=dyn_bandanna_l_03, swing1LimitDegrees=22, swing2LimitDegrees=45, twistLimitDegrees=3}
 {bone1Name=dyn_bandanna_l_07, bone2Name=dyn_bandanna_l_05, swing1LimitDegrees=22.5, swing2LimitDegrees=45, twistLimitDegrees=3}
 {bone1Name=dyn_bandanna_l_01, bone2Name=head, swing1LimitDegrees=3, swing2LimitDegrees=29.52, twistLimitDegrees=3}
 {bone1Name=dyn_bandanna_l_01, bone2Name=neck_02, swing1LimitDegrees=9, swing2LimitDegrees=26.28, twistLimitDegrees=5}
]
