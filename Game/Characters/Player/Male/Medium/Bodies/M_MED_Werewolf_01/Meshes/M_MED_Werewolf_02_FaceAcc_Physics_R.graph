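# Physics asset for the right-side bandanna face accessory on the medium male Werewolf 02 player character.
# Four box bodies form the dyn_bandanna_r chain, constrained link to link from the head with tight twist limits and swing mostly favoring one axis.
# Head, upper spine, clavicle and upper arm bodies have no constraints and supply surrounding collision volumes.
# The head pairs a small zero-length capsule with a large, far-offset sphere.
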PhysicsAsset /Game/Characters/Player/Male/Medium/Bodies/M_MED_Werewolf_01/Meshes/M_MED_Werewolf_02_FaceAcc_Physics_R
bodies {
 clavicle_l [{shapeType=Capsule, length=9.375, radius=7.5, center=<9.7252 0 0>, rotation=<rot -90 0.0003 -0.0002>}]
 clavicle_r [{shapeType=Capsule, length=9.375, radius=7.5, center=<-9.4582 0.0001 0>, rotation=<rot -90 0.0003 -0.0002>}]
 dyn_bandanna_r_01 [{shapeType=Box, extentX=5.75, extentY=4, extentZ=2, center=<4.4218 0 0.0006>}]
 dyn_bandanna_r_02 [{shapeType=Box, extentX=8.5, extentY=4.5, extentZ=2.25, center=<5.8885 -0.0001 0>}]
 dyn_bandanna_r_03 [{shapeType=Box, extentX=8.5, extentY=4.5, extentZ=3, center=<5.675 -0.0001 0>}]
 dyn_bandanna_r_04 [{shapeType=Box, extentX=6, extentY=4.5, extentZ=3, center=<5.205 0 0>}]
 head [
  {shapeType=Sphere, radius=50.5, center=<58.428 17.3581 0.0002>}
  {shapeType=Capsule, length=0, radius=11.875, center=<-2.8021 0.8878 -0.0001>}
 ]
 spine_05 [{shapeType=Capsule, length=0, radius=24.125, center=<0.5905 5.8427 -0.0001>}]
 upperarm_l [{shapeType=Capsule, length=27.375, radius=7, center=<13.1604 0.6274 -0.2085>, rotation=<rot -88.1986 -86.898 89.6603>}]
 upperarm_r [{shapeType=Capsule, length=27.375, radius=7, center=<-12.903 -0.6028 0.6643>, rotation=<rot -88.1987 -86.8981 89.6604>}]
}
constraints [
 {bone1Name=dyn_bandanna_r_01, bone2Name=head, swing1LimitDegrees=3, swing2LimitDegrees=90, twistLimitDegrees=3}
 {bone1Name=dyn_bandanna_r_02, bone2Name=dyn_bandanna_r_01, swing1LimitDegrees=3, swing2LimitDegrees=22.5, twistLimitDegrees=3}
 {bone1Name=dyn_bandanna_r_03, bone2Name=dyn_bandanna_r_02, swing1LimitDegrees=22.5, swing2LimitDegrees=45, twistLimitDegrees=3}
 {bone1Name=dyn_bandanna_r_04, bone2Name=dyn_bandanna_r_03, swing1LimitDegrees=22.5, swing2LimitDegrees=45, twistLimitDegrees=3}
]
